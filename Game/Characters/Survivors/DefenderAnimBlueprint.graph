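# DefenderAnimBlueprint is the animation blueprint for Defender survivor characters on the Fortnite player skeleton.
# Weapon-type, gender, targeting and relaxed-state blends choose poses, with aim offsets mostly driven by pitch.
# The SurvivorFullBody state machine is cached, layered with montage slots, then combined with the SurvivorAdditive machine.
# Hand IK retargeting, curve-weighted two-bone arm IK and a BoomBow weapon-bone copy finish the pose.
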AnimBlueprint /Game/Characters/Survivors/DefenderAnimBlueprint
target_skeleton=&Fortnite_M_Avg_Player_Skeleton
graphs [
 AnimGraph {nodes [
   Blend Poses by bool(BlendPose_1=Use cached pose 'FullBodyLocomotion'(), bActiveValue=Get bHasWeapon() -> (bHasWeapon:bool)) {blendType=Linear}
   Blend Poses by bool(BlendPose_1=Use cached pose 'FullBodyLocomotion'(), bActiveValue=NOT Boolean(A=Get bIsDBNO())) {blendType=Linear}
   Layered blend per bone(BlendPoses_0=Use cached pose 'FullBodyLocomotion'(), BlendWeights_0=Get LowerBodyCurrentWeight() -> (LowerBodyCurrentWeight:float)) {layerSetup [{branchFilters [{boneName=Root}, {blendDepth=-1, boneName=Ik_hand_root}, {blendDepth=-2, boneName=Spine_01}]}]}
   n1 = Layered blend per bone(
    BasePose=Slot 'PreAimOffsets'(Source=Use cached pose 'FullBodyStateMachine'()) {slotName=PreAimOffsets}
    BlendPoses_0=Slot 'FullBodyNoRootMotion'(Source=Use cached pose 'FullBodyStateMachine'()) {slotName=FullBodyNoRootMotion}
   ) {layerSetup [{branchFilters [{boneName=root}, {blendDepth=-4, boneName=spine_05}, {blendDepth=-1, boneName=ik_hand_root}]}]}
   n2 = Slot 'AdditiveLayer'(Source=Slot 'FullBody'(Source=@n1) {slotName=FullBody}) {slotName=AdditiveLayer}
   AnimGraphNode_StateMachine_38 = SurvivorAdditive() {is=title AnimGraphNode_StateMachine}
   n3 = Hand IK Retargeting!(
    ComponentPose:Component Space Pose Link Structure=Local To Component(LocalPose=Apply Additive(Base=@n2, Additive=@AnimGraphNode_StateMachine_38) {alphaBoolBlend {blendOption=Linear}})
    Alpha:float=1
    HandFKWeight:float=1
   ) -> (Pose:Component Space Pose Link Structure) {
    is=title AnimGraphNode_HandIKRetargeting
    lODThreshold=1
    alphaBoolBlend {blendOption=Linear}
    iKBonesToMove [{boneName=ik_hand_gun}]
    leftHandFK {boneName=hand_l}
    leftHandIK {boneName=ik_hand_l}
    rightHandFK {boneName=hand_r}
    rightHandIK {boneName=ik_hand_r}
   }
   n4 = float - float(B=GetCurveValue(CurveName=BothHandFK)) {is=title K2Node_CallFunction}
   n5 = Two Bone IK - Bone: hand_r!(ComponentPose:Component Space Pose Link Structure=@n3, Alpha:float=@n4, JointTargetLocation:Vector="100,100,0") -> (Pose:Component Space Pose Link Structure) {
    is=title AnimGraphNode_TwoBoneIK
    bAllowStretching=true
    bTakeRotationFromEffectorSpace=true
    effectorLocationSpace=BCS_BoneSpace
    jointTargetLocationSpace=BCS_BoneSpace
    maxStretchScale=1.25
    startStretchRatio=0.95
    alphaBoolBlend {blendOption=Linear}
    effectorTarget {boneReference {boneName=ik_hand_r}}
    iKBone {boneName=hand_r}
    jointTarget {boneReference {boneName=lowerarm_r}}
    jointTargetLocation=<100 100>
   }
   n6 = GetCurveValue(CurveName=LeftHandFK)
   n7 = Two Bone IK - Bone: hand_l!(
    ComponentPose:Component Space Pose Link Structure=@n5
    Alpha:float=float * float(A=@n4, B=float - float(B=@n6))
    JointTargetLocation:Vector="-100.000,-100.000,0.000"
   ) -> (Pose:Component Space Pose Link Structure) {
    is=title AnimGraphNode_TwoBoneIK
    bAllowStretching=true
    bTakeRotationFromEffectorSpace=true
    effectorLocationSpace=BCS_BoneSpace
    jointTargetLocationSpace=BCS_BoneSpace
    maxStretchScale=1.25
    startStretchRatio=0.95
    alphaBoolBlend {blendOption=Linear}
    effectorTarget {boneReference {boneName=ik_hand_l}}
    iKBone {boneName=hand_l}
    jointTarget {boneReference {boneName=lowerarm_l}}
    jointTargetLocation=<-100 -100>
   }
   n8 = Copy Bone - Source Bone: weapon_l - Target Bone: weapon_r!(
    ComponentPose:Component Space Pose Link Structure=@n7
    bAlphaBoolEnabled:bool=Get UsingBoomBow() -> (UsingBoomBow:bool)
    bCopyTranslation:bool=true
    bCopyRotation:bool=true
    bCopyScale:bool=true
   ) -> (Pose:Component Space Pose Link Structure) {
    is=title AnimGraphNode_CopyBone
    alphaInputType=Bool
    alphaBoolBlend {blendOption=Linear}
    sourceBone {boneName=weapon_l}
    targetBone {boneName=weapon_r}
   }
   Output Pose(Result=Component To Local(ComponentPose=@n8)) {layerGroup=None}
   Rotate Root Bone!(BasePose:Pose Link Structure, Pitch:float=0, Yaw:float=0, MeshToComponent:Rotator="0,0,0") -> (Pose:Pose Link Structure) {is=title AnimGraphNode_RotateRootBone}
   AnimGraphNode_StateMachine_55 = SurvivorFullBody() {is=title AnimGraphNode_StateMachine}
   Save cached pose 'FullBodyStateMachine'(Pose=@AnimGraphNode_StateMachine_55) {cacheName=FullBodyStateMachine}
   n9 = "Blend Poses (EFortCustomGender)"!(
    BlendPose_0:Pose Link Structure=Sequence Player 'M_NoWep_Relaxed'() {sequence=/Game/Animation/Game/MainPlayer/Poses/Medium/Male/M_NoWep_Relaxed}
    BlendPose_1:Pose Link Structure=Sequence Player 'Pose_NoWep_Relaxed_CMF'() {sequence=/Game/Animation/Game/MainPlayer/Poses/Medium/Female/Pose_NoWep_Relaxed_CMF}
    BlendTime_0:float=0
    BlendTime_1:float=0
    ActiveEnumValue:Byte=Get Gender() -> (Gender:EFortCustomGender Enum=Invalid)
   ) -> (Pose:Pose Link Structure) {
    is=title AnimGraphNode_BlendListByEnum
    blendType=Linear
    blendPose [{linkId=-1, sourceLinkId=-1}, {linkId=-1, sourceLinkId=-1}]
    blendTime [0.1,0.1]
   }
   n10 = "Blend Poses (EFortCustomGender)"(
    BlendPose_0=Sequence Player 'AR_Relaxed_Pose_CMM'() {sequence=/Game/Animation/Game/MainPlayer/Poses/Medium/Male/AssaultRifle/AR_Relaxed_Pose_CMM}
    BlendPose_1=Sequence Player 'AR_Relaxed_Pose_CMF'() {sequence=/Game/Animation/Game/MainPlayer/Poses/Medium/Female/AssaultRifle/AR_Relaxed_Pose_CMF}
    BlendTime_0=0.1
    BlendTime_1=0.1
    ActiveEnumValue=Get Gender() -> (Gender=Invalid)
   ) {blendType=Linear, blendPose [{linkId=-1, sourceLinkId=-1}, {linkId=-1, sourceLinkId=-1}], blendTime [0.1,0.1]}
   n11 = float * float(A=float - float(B=GetCurveValue(CurveName=ZeroOutPitchWeight)), B=Get AimPitch())
   n12 = AimOffset Player 'M_NoWep_Relaxed_AO'!(X:float=0, Y:float=@n11, BasePose:Pose Link Structure=@n10, Alpha:float=1) -> (Pose:Pose Link Structure) {blendSpace=/Game/Animation/Game/MainPlayer/Poses/Medium/Male/AO/M_NoWep_Relaxed_AO, alphaBoolBlend {blendOption=Linear}}
   n13 = "Blend Poses (EFortCustomGender)"(
    BlendPose_0=Sequence Player 'M_AssaultRifle_HipNonTargeted'() {sequence=/Game/Animation/Game/MainPlayer/Poses/Medium/Male/M_AssaultRifle_HipNonTargeted}
    BlendPose_1=Sequence Player 'Pose_AssaultRifle_HipFireNonTargeted_CMF'() {sequence=/Game/Animation/Game/MainPlayer/Poses/Medium/Female/Pose_AssaultRifle_HipFireNonTargeted_CMF}
    BlendTime_0=0.1
    BlendTime_1=0.1
    ActiveEnumValue=Get Gender() -> (Gender=Invalid)
   ) {blendType=Linear, blendPose [{linkId=-1, sourceLinkId=-1}, {linkId=-1, sourceLinkId=-1}], blendTime [0.1,0.1]}
   n14 = AimOffset Player 'M_AO_AR_Downsights'!(X:float=0, Y:float=@n11, BasePose:Pose Link Structure=@n13, Alpha:float=1) -> (Pose:Pose Link Structure) {blendSpace=/Game/Animation/Game/MainPlayer/Combat/Medium/Guns/AssaultRifle/M_AO_AR_Downsights, alphaBoolBlend {blendOption=Linear}}
   n15 = OR Boolean(A=Get bIsRelaxedLevel1() -> (bIsRelaxedLevel1:bool), B=NOT Boolean(A=Get bIsTargeting()))
   n16 = "Blend Poses (EFortCustomGender)"(
    BlendPose_0=Sequence Player 'AR_Targeted_Pose_CMM'() {sequence=/Game/Animation/Game/MainPlayer/Poses/Medium/Male/AssaultRifle/AR_Targeted_Pose_CMM}
    BlendPose_1=Sequence Player 'AR_Targeted_Pose_CMF'() {sequence=/Game/Animation/Game/MainPlayer/Poses/Medium/Female/AssaultRifle/AR_Targeted_Pose_CMF}
    BlendTime_0=0.1
    BlendTime_1=0.1
    ActiveEnumValue=Get Gender() -> (Gender=Invalid)
   ) {blendType=Linear, blendPose [{linkId=-1, sourceLinkId=-1}, {linkId=-1, sourceLinkId=-1}], blendTime [0.1,0.1]}
   n17 = AimOffset Player 'M_AO_AR_Downsights'(X=0, Y=@n11, BasePose=@n16, Alpha=1) {blendSpace=&M_AO_AR_Downsights, alphaBoolBlend {blendOption=Linear}}
   n18 = "Blend Poses (EFortCustomGender)"(
    BlendPose_0=Sequence Player 'AR_Relaxed_Pose_CMM'() {sequence=&AR_Relaxed_Pose_CMM}
    BlendPose_1=Sequence Player 'AR_Relaxed_Pose_CMF'() {sequence=&AR_Relaxed_Pose_CMF}
    BlendTime_0=0.1
    BlendTime_1=0.1
    ActiveEnumValue=Get Gender() -> (Gender=Invalid)
   ) {blendType=Linear, blendPose [{linkId=-1, sourceLinkId=-1}, {linkId=-1, sourceLinkId=-1}], blendTime [0.1,0.1]}
   n19 = float * float(A=float - float(B=GetCurveValue(CurveName=ZeroOutPitchWeight)), B=Get AimPitch())
   n20 = AimOffset Player 'M_NoWep_Relaxed_AO'(X=0, Y=@n19, BasePose=@n18, Alpha=1) {blendSpace=&M_NoWep_Relaxed_AO, alphaBoolBlend {blendOption=Linear}}
   n21 = "Blend Poses (EFortCustomGender)"(
    BlendPose_0=Sequence Player 'AR_NonTargeted_Pose_CMM'() {sequence=/Game/Animation/Game/MainPlayer/Poses/Medium/Male/AssaultRifle/AR_NonTargeted_Pose_CMM}
    BlendPose_1=Sequence Player 'Crossbow_NonTargeted_CMM'() {sequence=/Game/Animation/Game/MainPlayer/Poses/Medium/Male/Crossbow_NonTargeted_CMM}
    BlendTime_0=0.1
    BlendTime_1=0.1
    ActiveEnumValue=Get Gender() -> (Gender=Invalid)
   ) {blendType=Linear, blendPose [{linkId=-1, sourceLinkId=-1}, {linkId=-1, sourceLinkId=-1}], blendTime [0.1,0.1]}
   n22 = AimOffset Player 'M_AO_AR_Downsights'(X=0, Y=@n19, BasePose=@n21, Alpha=1) {blendSpace=&M_AO_AR_Downsights, alphaBoolBlend {blendOption=Linear}}
   n23 = OR Boolean(A=Get bIsRelaxedLevel1(), B=NOT Boolean(A=Get bIsTargeting()))
   n24 = "Blend Poses (EFortCustomGender)"(
    BlendPose_0=Sequence Player 'AR_Targeted_Pose_CMM'() {sequence=&AR_Targeted_Pose_CMM}
    BlendPose_1=Sequence Player 'AR_Targeted_Pose_CMF'() {sequence=&AR_Targeted_Pose_CMF}
    BlendTime_0=0.1
    BlendTime_1=0.1
    ActiveEnumValue=Get Gender() -> (Gender=Invalid)
   ) {blendType=Linear, blendPose [{linkId=-1, sourceLinkId=-1}, {linkId=-1, sourceLinkId=-1}], blendTime [0.1,0.1]}
   n25 = AimOffset Player 'M_AO_AR_Downsights'(X=0, Y=@n19, BasePose=@n24, Alpha=1) {blendSpace=&M_AO_AR_Downsights, alphaBoolBlend {blendOption=Linear}}
   n26 = "Blend Poses (EFortCustomGender)"(
    BlendPose_0=Sequence Player 'Pose_ARFGrip_Relaxed_CMM'() {sequence=/Game/Animation/Game/MainPlayer/Poses/Medium/Male/AR_ForwardGrip/Pose_ARFGrip_Relaxed_CMM}
    BlendPose_1=Sequence Player 'Pose_ARFGrip_Relaxed_CMF'() {sequence=/Game/Animation/Game/MainPlayer/Poses/Medium/Female/AR_ForwardGrip/Pose_ARFGrip_Relaxed_CMF}
    BlendTime_0=0.1
    BlendTime_1=0.1
    ActiveEnumValue=Get Gender() -> (Gender=Invalid)
   ) {blendType=Linear, blendPose [{linkId=-1, sourceLinkId=-1}, {linkId=-1, sourceLinkId=-1}], blendTime [0.1,0.1]}
   n27 = float * float(A=float - float(B=GetCurveValue(CurveName=ZeroOutPitchWeight)), B=Get AimPitch())
   n28 = AimOffset Player 'M_NoWep_Relaxed_AO'(X=0, Y=@n27, BasePose=@n26, Alpha=1) {blendSpace=&M_NoWep_Relaxed_AO, alphaBoolBlend {blendOption=Linear}}
   n29 = "Blend Poses (EFortCustomGender)"(
    BlendPose_0=Sequence Player 'Pose_ARFGrip_NonTargeted_CMM'() {sequence=/Game/Animation/Game/MainPlayer/Poses/Medium/Male/AR_ForwardGrip/Pose_ARFGrip_NonTargeted_CMM}
    BlendPose_1=Sequence Player 'Pose_ARFGrip_NonTargeted_CMF'() {sequence=/Game/Animation/Game/MainPlayer/Poses/Medium/Female/AR_ForwardGrip/Pose_ARFGrip_NonTargeted_CMF}
    BlendTime_0=0.1
    BlendTime_1=0.1
    ActiveEnumValue=Get Gender() -> (Gender=Invalid)
   ) {blendType=Linear, blendPose [{linkId=-1, sourceLinkId=-1}, {linkId=-1, sourceLinkId=-1}], blendTime [0.1,0.1]}
   n30 = AimOffset Player 'M_AO_AR_Downsights'(X=0, Y=@n27, BasePose=@n29, Alpha=1) {blendSpace=&M_AO_AR_Downsights, alphaBoolBlend {blendOption=Linear}}
   n31 = OR Boolean(A=Get bIsRelaxedLevel1(), B=NOT Boolean(A=Get bIsTargeting()))
   n32 = "Blend Poses (EFortCustomGender)"(
    BlendPose_0=Sequence Player 'Pose_ARFGrip_Targeted_CMM'() {sequence=/Game/Animation/Game/MainPlayer/Poses/Medium/Male/AR_ForwardGrip/Pose_ARFGrip_Targeted_CMM}
    BlendPose_1=Sequence Player 'Pose_ARFGrip_Targeted_CMF'() {sequence=/Game/Animation/Game/MainPlayer/Poses/Medium/Female/AR_ForwardGrip/Pose_ARFGrip_Targeted_CMF}
    BlendTime_0=0.1
    BlendTime_1=0.1
    ActiveEnumValue=Get Gender() -> (Gender=Invalid)
   ) {blendType=Linear, blendPose [{linkId=-1, sourceLinkId=-1}, {linkId=-1, sourceLinkId=-1}], blendTime [0.1,0.1]}
   n33 = AimOffset Player 'M_AO_AR_Downsights'(X=0, Y=@n27, BasePose=@n32, Alpha=1) {blendSpace=&M_AO_AR_Downsights, alphaBoolBlend {blendOption=Linear}}
   n34 = "Blend Poses (EFortCustomGender)"(
    BlendPose_0=Sequence Player 'AR_Relaxed_Pose_CMM'() {sequence=&AR_Relaxed_Pose_CMM}
    BlendPose_1=Sequence Player 'AR_Relaxed_Pose_CMF'() {sequence=&AR_Relaxed_Pose_CMF}
    BlendTime_0=0.1
    BlendTime_1=0.1
    ActiveEnumValue=Get Gender() -> (Gender=Invalid)
   ) {blendType=Linear, blendPose [{linkId=-1, sourceLinkId=-1}, {linkId=-1, sourceLinkId=-1}], blendTime [0.1,0.1]}
   n35 = float * float(A=float - float(B=GetCurveValue(CurveName=ZeroOutPitchWeight)), B=Get AimPitch())
   n36 = AimOffset Player 'M_NoWep_Relaxed_AO'(X=0, Y=@n35, BasePose=@n34, Alpha=1) {blendSpace=&M_NoWep_Relaxed_AO, alphaBoolBlend {blendOption=Linear}}
   n37 = "Blend Poses (EFortCustomGender)"(
    BlendPose_0=Sequence Player 'AR_NonTargeted_Pose_CMM'() {sequence=&AR_NonTargeted_Pose_CMM}
    BlendPose_1=Sequence Player 'AR_NonTargeted_Pose_CMF'() {sequence=/Game/Animation/Game/MainPlayer/Poses/Medium/Female/AssaultRifle/AR_NonTargeted_Pose_CMF}
    BlendTime_0=0.1
    BlendTime_1=0.1
    ActiveEnumValue=Get Gender() -> (Gender=Invalid)
   ) {blendType=Linear, blendPose [{linkId=-1, sourceLinkId=-1}, {linkId=-1, sourceLinkId=-1}], blendTime [0.1,0.1]}
   n38 = AimOffset Player 'M_AO_AR_Downsights'(X=0, Y=@n35, BasePose=@n37, Alpha=1) {blendSpace=&M_AO_AR_Downsights, alphaBoolBlend {blendOption=Linear}}
   n39 = OR Boolean(A=Get bIsRelaxedLevel1(), B=NOT Boolean(A=Get bIsTargeting()))
   n40 = "Blend Poses (EFortCustomGender)"(
    BlendPose_0=Sequence Player 'AR_Targeted_Pose_CMM'() {sequence=&AR_Targeted_Pose_CMM}
    BlendPose_1=Sequence Player 'AR_Targeted_Pose_CMF'() {sequence=&AR_Targeted_Pose_CMF}
    BlendTime_0=0.1
    BlendTime_1=0.1
    ActiveEnumValue=Get Gender() -> (Gender=Invalid)
   ) {blendType=Linear, blendPose [{linkId=-1, sourceLinkId=-1}, {linkId=-1, sourceLinkId=-1}], blendTime [0.1,0.1]}
   n41 = AimOffset Player 'M_AO_AR_Downsights'(X=0, Y=@n35, BasePose=@n40, Alpha=1) {blendSpace=&M_AO_AR_Downsights, alphaBoolBlend {blendOption=Linear}}
   n42 = "Blend Poses (EFortCustomGender)"(
    BlendPose_0=Sequence Player 'M_Pistol_HF'() {sequence=/Game/Animation/Game/MainPlayer/Poses/Medium/Male/M_Pistol_HF}
    BlendPose_1=Sequence Player 'Pose_Pistol_NonTargeted_CMF'() {sequence=/Game/Animation/Game/MainPlayer/Poses/Medium/Female/Pose_Pistol_NonTargeted_CMF}
    BlendTime_0=0.1
    BlendTime_1=0.1
    ActiveEnumValue=Get Gender() -> (Gender=Invalid)
   ) {blendType=Linear, blendPose [{linkId=-1, sourceLinkId=-1}, {linkId=-1, sourceLinkId=-1}], blendTime [0.1,0.1]}
   n43 = float * float(A=float - float(B=GetCurveValue(CurveName=ZeroOutPitchWeight)), B=Get AimPitch())
   n44 = AimOffset Player 'M_AO_Pistol_Downsights'!(X:float=0, Y:float=@n43, BasePose:Pose Link Structure=@n42, Alpha:float=1) -> (Pose:Pose Link Structure) {blendSpace=/Game/Animation/Game/MainPlayer/Combat/Medium/Guns/Pistol/M_AO_Pistol_Downsights, alphaBoolBlend {blendOption=Linear}}
   n45 = "Blend Poses (EFortCustomGender)"(
    BlendPose_0=Sequence Player 'M_Pistol_Relaxed'() {sequence=/Game/Animation/Game/MainPlayer/Poses/Medium/Male/M_Pistol_Relaxed}
    BlendPose_1=Sequence Player 'Pose_Pistol_Relaxed_CMF'() {sequence=/Game/Animation/Game/MainPlayer/Poses/Medium/Female/Pose_Pistol_Relaxed_CMF}
    BlendTime_0=0.1
    BlendTime_1=0.1
    ActiveEnumValue=Get Gender() -> (Gender=Invalid)
   ) {blendType=Linear, blendPose [{linkId=-1, sourceLinkId=-1}, {linkId=-1, sourceLinkId=-1}], blendTime [0.1,0.1]}
   n46 = "Blend Poses (EFortCustomGender)"(
    BlendPose_0=Sequence Player 'M_Pistol_Relaxed2'() {sequence=/Game/Animation/Game/MainPlayer/Poses/Medium/Male/M_Pistol_Relaxed2}
    BlendPose_1=Sequence Player 'Pose_Pistol_Relaxed2_CMF'() {sequence=/Game/Animation/Game/MainPlayer/Poses/Medium/Female/Pose_Pistol_Relaxed2_CMF}
    BlendTime_0=0.1
    BlendTime_1=0.1
    ActiveEnumValue=Get Gender() -> (Gender=Invalid)
   ) {blendType=Linear, blendPose [{linkId=-1, sourceLinkId=-1}, {linkId=-1, sourceLinkId=-1}], blendTime [0.1,0.1]}
   n47 = AimOffset Player 'M_NoWep_Relaxed_AO'(
    X=0
    Y=@n43
    BasePose=Blend Poses by bool(BlendPose_0=@n45, BlendPose_1=@n46, BlendTime_0=.4, BlendTime_1=0, bActiveValue=Get bIsRelaxedLevel2() -> (bIsRelaxedLevel2:bool)) {blendType=Linear}
    Alpha=1
   ) {blendSpace=&M_NoWep_Relaxed_AO, alphaBoolBlend {blendOption=Linear}}
   n48 = Get bIsRelaxedLevel1()
   n49 = "Blend Poses (EFortCustomGender)"(
    BlendPose_0=Sequence Player 'M_Pistol_DS'() {sequence=/Game/Animation/Game/MainPlayer/Poses/Medium/Male/M_Pistol_DS}
    BlendPose_1=Sequence Player 'Pose_Pistol_Targeted_CMF'() {sequence=/Game/Animation/Game/MainPlayer/Poses/Medium/Female/Pose_Pistol_Targeted_CMF}
    BlendTime_0=0.1
    BlendTime_1=0.1
    ActiveEnumValue=Get Gender() -> (Gender=Invalid)
   ) {blendType=Linear, blendPose [{linkId=-1, sourceLinkId=-1}, {linkId=-1, sourceLinkId=-1}], blendTime [0.1,0.1]}
   n50 = Get AimYaw() -> (AimYaw:float)
   n51 = "Blend Poses (EFortCustomGender)"(
    BlendPose_0=Sequence Player 'Rifle_Relaxed_CMM'() {sequence=/Game/Animation/Game/MainPlayer/Poses/Medium/Male/Rifle_Relaxed_CMM}
    BlendPose_1=Sequence Player 'Pose_Rifle_Relaxed_CMF'() {sequence=/Game/Animation/Game/MainPlayer/Poses/Medium/Female/Pose_Rifle_Relaxed_CMF}
    BlendTime_0=0.1
    BlendTime_1=0.1
    ActiveEnumValue=Get Gender() -> (Gender=Invalid)
   ) {blendType=Linear, blendPose [{linkId=-1, sourceLinkId=-1}, {linkId=-1, sourceLinkId=-1}], blendTime [0.1,0.1]}
   n52 = AimOffset Player 'M_NoWep_Relaxed_AO'(X=0, Y=@n43, BasePose=@n51, Alpha=1) {blendSpace=&M_NoWep_Relaxed_AO, alphaBoolBlend {blendOption=Linear}}
   n53 = "Blend Poses (EFortCustomGender)"(
    BlendPose_0=Sequence Player 'Rifle_HF_CMM'() {sequence=/Game/Animation/Game/MainPlayer/Poses/Medium/Male/Rifle_HF_CMM}
    BlendPose_1=Sequence Player 'Pose_Rifle_HipNonTargeted_CMF'() {sequence=/Game/Animation/Game/MainPlayer/Poses/Medium/Female/Pose_Rifle_HipNonTargeted_CMF}
    BlendTime_0=0.1
    BlendTime_1=0.1
    ActiveEnumValue=Get Gender() -> (Gender=Invalid)
   ) {blendType=Linear, blendPose [{linkId=-1, sourceLinkId=-1}, {linkId=-1, sourceLinkId=-1}], blendTime [0.1,0.1]}
   n54 = AimOffset Player 'M_AO_AR_Downsights'(X=0, Y=@n43, BasePose=@n53, Alpha=1) {blendSpace=&M_AO_AR_Downsights, alphaBoolBlend {blendOption=Linear}}
   n55 = Get bIsRelaxedLevel1()
   n56 = "Blend Poses (EFortCustomGender)"(
    BlendPose_0=Sequence Player 'Rifle_Targeted_CMM'() {sequence=/Game/Animation/Game/MainPlayer/Poses/Medium/Male/Rifle_Targeted_CMM}
    BlendPose_1=Sequence Player 'Pose_Rifle_Targeted_CMF'() {sequence=/Game/Animation/Game/MainPlayer/Poses/Medium/Female/Pose_Rifle_Targeted_CMF}
    BlendTime_0=0.1
    BlendTime_1=0.1
    ActiveEnumValue=Get Gender() -> (Gender=Invalid)
   ) {blendType=Linear, blendPose [{linkId=-1, sourceLinkId=-1}, {linkId=-1, sourceLinkId=-1}], blendTime [0.1,0.1]}
   n57 = AimOffset Player 'M_AO_AR_Downsights'(X=0, Y=@n43, BasePose=@n56, Alpha=1) {blendSpace=&M_AO_AR_Downsights, alphaBoolBlend {blendOption=Linear}}
   n58 = "Blend Poses (EFortCustomGender)"(
    BlendPose_0="Sequence Player (additive) 'ExplosiveBow_ChargingLoop_CMM'"() {sequence=/Game/Animation/Game/MainPlayer/Combat/Guns/ExplosiveBow/ExplosiveBow_ChargingLoop_CMM}
    BlendPose_1="Sequence Player (additive) 'ExplosiveBow_ChargingLoop_CMF'"() {sequence=/Game/Animation/Game/MainPlayer/Combat/Guns/ExplosiveBow/ExplosiveBow_ChargingLoop_CMF}
    BlendTime_0=0.1
    BlendTime_1=0.1
    ActiveEnumValue=Get Gender() -> (Gender=Invalid)
   ) {blendType=Linear, blendPose [{linkId=-1, sourceLinkId=-1}, {linkId=-1, sourceLinkId=-1}], blendTime [0.1,0.1]}
   n59 = Get AimPitch()
   n60 = "Blend Poses (EFortCustomGender)"(
    BlendPose_0=Blendspace Player 'ExplosiveBow_Combined_Charging_BS_CMM'!(X=1, Y=@n59, PlayRate:float=1) -> (Pose) {bLoop=false, blendSpace=/Game/Animation/Game/MainPlayer/Combat/Guns/ExplosiveBow/ExplosiveBow_Combined_Charging_BS_CMM}
    BlendPose_1=Blendspace Player 'ExplosiveBow_Combined_Charging_BS_CMF'!(X=1, Y=@n59, PlayRate:float=1) -> (Pose) {bLoop=false, blendSpace=/Game/Animation/Game/MainPlayer/Combat/Guns/ExplosiveBow/ExplosiveBow_Combined_Charging_BS_CMF}
    BlendTime_0=0.1
    BlendTime_1=0.1
    ActiveEnumValue=Get Gender() -> (Gender=Invalid)
   ) {blendType=Linear, blendPose [{linkId=-1, sourceLinkId=-1}, {linkId=-1, sourceLinkId=-1}], blendTime [0.1,0.1]}
   n61 = Apply Additive!(Base=@n60, Additive=@n58, bAlphaBoolEnabled:bool=True) -> (Pose) {alphaInputType=Bool, alphaBoolBlend {blendOption=Linear}}
   n62 = "Blend Poses (EFortCustomGender)"(
    BlendPose_0=Sequence Player 'ExplosiveBow_NonTargeting_Pose_CMM'() {sequence=/Game/Animation/Game/MainPlayer/Poses/Medium/Male/ExplosiveBow/ExplosiveBow_NonTargeting_Pose_CMM}
    BlendPose_1=Sequence Player 'ExplosiveBow_NonTargeting_Pose_CMF'() {sequence=/Game/Animation/Game/MainPlayer/Poses/Medium/Female/ExplosiveBow/ExplosiveBow_NonTargeting_Pose_CMF}
    BlendTime_0=0.1
    BlendTime_1=0.1
    ActiveEnumValue=Get Gender() -> (Gender=Invalid)
   ) {blendType=Linear, blendPose [{linkId=-1, sourceLinkId=-1}, {linkId=-1, sourceLinkId=-1}], blendTime [0.1,0.1]}
   n63 = AimOffset Player 'M_AO_AR_Downsights'(X=0, Y=0, BasePose=@n62, Alpha=1) {blendSpace=&M_AO_AR_Downsights, alphaBoolBlend {blendOption=Linear}}
   n64 = Get bIsChargingWeapon() -> (bIsChargingWeapon:bool)
   n65 = "Blend Poses (EFortCustomGender)"(
    BlendPose_0=Sequence Player 'ExplosiveBow_Relaxed_Pose_CMM'() {sequence=/Game/Animation/Game/MainPlayer/Poses/Medium/Male/ExplosiveBow/ExplosiveBow_Relaxed_Pose_CMM}
    BlendPose_1=Sequence Player 'ExplosiveBow_Relaxed_Pose_CMF'() {sequence=/Game/Animation/Game/MainPlayer/Poses/Medium/Female/ExplosiveBow/ExplosiveBow_Relaxed_Pose_CMF}
    BlendTime_0=0.1
    BlendTime_1=0.1
    ActiveEnumValue=Get Gender() -> (Gender=Invalid)
   ) {blendType=Linear, blendPose [{linkId=-1, sourceLinkId=-1}, {linkId=-1, sourceLinkId=-1}], blendTime [0.1,0.1]}
   n66 = float * float(A=float - float(B=GetCurveValue(CurveName=ZeroOutPitchWeight)), B=@n59)
   n67 = NOT Boolean(A=Get bIsTargeting())
   n68 = Get bIsRelaxedLevel1()
   n69 = "Blend Poses (EFortCustomGender)"(
    BlendPose_0=Sequence Player 'Pose_PDW_Relaxed_CMM'() {sequence=/Game/Animation/Game/MainPlayer/Poses/Medium/Male/P90/Pose_PDW_Relaxed_CMM}
    BlendPose_1=Sequence Player 'Pose_PDW_Relaxed_CMF'() {sequence=/Game/Animation/Game/MainPlayer/Poses/Medium/Female/P90/Pose_PDW_Relaxed_CMF}
    BlendTime_0=0.1
    BlendTime_1=0.1
    ActiveEnumValue=Get Gender() -> (Gender=Invalid)
   ) {blendType=Linear, blendPose [{linkId=-1, sourceLinkId=-1}, {linkId=-1, sourceLinkId=-1}], blendTime [0.1,0.1]}
   n70 = float * float(A=float - float(B=GetCurveValue(CurveName=ZeroOutPitchWeight)), B=Get AimPitch())
   n71 = AimOffset Player 'M_NoWep_Relaxed_AO'(X=0, Y=@n70, BasePose=@n69, Alpha=1) {blendSpace=&M_NoWep_Relaxed_AO, alphaBoolBlend {blendOption=Linear}}
   n72 = "Blend Poses (EFortCustomGender)"(
    BlendPose_0=Sequence Player 'Pose_PDW_NonTargeted_CMM'() {sequence=/Game/Animation/Game/MainPlayer/Poses/Medium/Male/P90/Pose_PDW_NonTargeted_CMM}
    BlendPose_1=Sequence Player 'Pose_PDW_NonTargeted_CMF'() {sequence=/Game/Animation/Game/MainPlayer/Poses/Medium/Female/P90/Pose_PDW_NonTargeted_CMF}
    BlendTime_0=0.1
    BlendTime_1=0.1
    ActiveEnumValue=Get Gender() -> (Gender=Invalid)
   ) {blendType=Linear, blendPose [{linkId=-1, sourceLinkId=-1}, {linkId=-1, sourceLinkId=-1}], blendTime [0.1,0.1]}
   n73 = AimOffset Player 'M_AO_AR_Downsights'(X=0, Y=@n70, BasePose=@n72, Alpha=1) {blendSpace=&M_AO_AR_Downsights, alphaBoolBlend {blendOption=Linear}}
   n74 = OR Boolean(A=Get bIsRelaxedLevel1(), B=NOT Boolean(A=Get bIsTargeting()))
   n75 = "Blend Poses (EFortCustomGender)"(
    BlendPose_0=Sequence Player 'Pose_PDW_Targeted_CMM'() {sequence=/Game/Animation/Game/MainPlayer/Poses/Medium/Male/P90/Pose_PDW_Targeted_CMM}
    BlendPose_1=Sequence Player 'Pose_PDW_Targeted_CMF'() {sequence=/Game/Animation/Game/MainPlayer/Poses/Medium/Female/P90/Pose_PDW_Targeted_CMF}
    BlendTime_0=0.1
    BlendTime_1=0.1
    ActiveEnumValue=Get Gender() -> (Gender=Invalid)
   ) {blendType=Linear, blendPose [{linkId=-1, sourceLinkId=-1}, {linkId=-1, sourceLinkId=-1}], blendTime [0.1,0.1]}
   n76 = AimOffset Player 'M_AO_AR_Downsights'(X=0, Y=@n70, BasePose=@n75, Alpha=1) {blendSpace=&M_AO_AR_Downsights, alphaBoolBlend {blendOption=Linear}}
   n77 = "Blend Poses (EFortCustomGender)"(
    BlendPose_0=Sequence Player 'Pose_AutoDrum_Relaxed_CMM'() {sequence=/Game/Animation/Game/MainPlayer/Poses/Medium/Male/AutoDrum/Pose_AutoDrum_Relaxed_CMM}
    BlendPose_1=Sequence Player 'Pose_AutoDrum_Relaxed_CMF'() {sequence=/Game/Animation/Game/MainPlayer/Poses/Medium/Female/AutoDrum/Pose_AutoDrum_Relaxed_CMF}
    BlendTime_0=0.1
    BlendTime_1=0.1
    ActiveEnumValue=Get Gender() -> (Gender=Invalid)
   ) {blendType=Linear, blendPose [{linkId=-1, sourceLinkId=-1}, {linkId=-1, sourceLinkId=-1}], blendTime [0.1,0.1]}
   n78 = float * float(A=float - float(B=GetCurveValue(CurveName=ZeroOutPitchWeight)), B=Get AimPitch())
   n79 = AimOffset Player 'M_NoWep_Relaxed_AO'(X=0, Y=@n78, BasePose=@n77, Alpha=1) {blendSpace=&M_NoWep_Relaxed_AO, alphaBoolBlend {blendOption=Linear}}
   n80 = "Blend Poses (EFortCustomGender)"(
    BlendPose_0=Sequence Player 'Pose_AutoDrum_NonTargeted_CMM'() {sequence=/Game/Animation/Game/MainPlayer/Poses/Medium/Male/AutoDrum/Pose_AutoDrum_NonTargeted_CMM}
    BlendPose_1=Sequence Player 'Pose_AutoDrum_NonTargeted_CMF'() {sequence=/Game/Animation/Game/MainPlayer/Poses/Medium/Female/AutoDrum/Pose_AutoDrum_NonTargeted_CMF}
    BlendTime_0=0.1
    BlendTime_1=0.1
    ActiveEnumValue=Get Gender() -> (Gender=Invalid)
   ) {blendType=Linear, blendPose [{linkId=-1, sourceLinkId=-1}, {linkId=-1, sourceLinkId=-1}], blendTime [0.1,0.1]}
   n81 = AimOffset Player 'M_AO_AR_Downsights'(X=0, Y=@n78, BasePose=@n80, Alpha=1) {blendSpace=&M_AO_AR_Downsights, alphaBoolBlend {blendOption=Linear}}
   n82 = OR Boolean(A=Get bIsRelaxedLevel1(), B=NOT Boolean(A=Get bIsTargeting()))
   n83 = "Blend Poses (EFortCustomGender)"(
    BlendPose_0=Sequence Player 'Pose_AutoDrum_Targeted_CMM'() {sequence=/Game/Animation/Game/MainPlayer/Poses/Medium/Male/AutoDrum/Pose_AutoDrum_Targeted_CMM}
    BlendPose_1=Sequence Player 'Pose_AutoDrum_Targeted_CMF'() {sequence=/Game/Animation/Game/MainPlayer/Poses/Medium/Female/AutoDrum/Pose_AutoDrum_Targeted_CMF}
    BlendTime_0=0.1
    BlendTime_1=0.1
    ActiveEnumValue=Get Gender() -> (Gender=Invalid)
   ) {blendType=Linear, blendPose [{linkId=-1, sourceLinkId=-1}, {linkId=-1, sourceLinkId=-1}], blendTime [0.1,0.1]}
   n84 = AimOffset Player 'M_AO_AR_Downsights'(X=0, Y=@n78, BasePose=@n83, Alpha=1) {blendSpace=&M_AO_AR_Downsights, alphaBoolBlend {blendOption=Linear}}
   n85 = "Blend Poses (EFortCustomGender)"(
    BlendPose_0=Sequence Player 'Rifle_Relaxed_CMM'() {sequence=&Rifle_Relaxed_CMM}
    BlendPose_1=Sequence Player 'Pose_Rifle_Relaxed_CMF'() {sequence=&Pose_Rifle_Relaxed_CMF}
    BlendTime_0=0.1
    BlendTime_1=0.1
    ActiveEnumValue=Get Gender() -> (Gender=Invalid)
   ) {blendType=Linear, blendPose [{linkId=-1, sourceLinkId=-1}, {linkId=-1, sourceLinkId=-1}], blendTime [0.1,0.1]}
   n86 = AimOffset Player 'M_NoWep_Relaxed_AO'(X=0, Y=@n43, BasePose=@n85, Alpha=1) {blendSpace=&M_NoWep_Relaxed_AO, alphaBoolBlend {blendOption=Linear}}
   n87 = "Blend Poses (EFortCustomGender)"(
    BlendPose_0=Sequence Player 'Rifle_NonTargeted_CMM'() {sequence=/Game/Animation/Game/MainPlayer/Poses/Medium/Male/Rifle_NonTargeted_CMM}
    BlendPose_1=Sequence Player 'Pose_Rifle_NonTargeted_CMF'() {sequence=/Game/Animation/Game/MainPlayer/Poses/Medium/Female/Pose_Rifle_NonTargeted_CMF}
    BlendTime_0=0.1
    BlendTime_1=0.1
    ActiveEnumValue=Get Gender() -> (Gender=Invalid)
   ) {blendType=Linear, blendPose [{linkId=-1, sourceLinkId=-1}, {linkId=-1, sourceLinkId=-1}], blendTime [0.1,0.1]}
   n88 = AimOffset Player 'M_AO_AR_Downsights'(X=0, Y=@n43, BasePose=@n87, Alpha=1) {blendSpace=&M_AO_AR_Downsights, alphaBoolBlend {blendOption=Linear}}
   n89 = Get bIsRelaxedLevel1()
   n90 = "Blend Poses (EFortCustomGender)"(
    BlendPose_0=Sequence Player 'Rifle_Targeted_CMM'() {sequence=&Rifle_Targeted_CMM}
    BlendPose_1=Sequence Player 'Pose_Rifle_Targeted_CMF'() {sequence=&Pose_Rifle_Targeted_CMF}
    BlendTime_0=0.1
    BlendTime_1=0.1
    ActiveEnumValue=Get Gender() -> (Gender=Invalid)
   ) {blendType=Linear, blendPose [{linkId=-1, sourceLinkId=-1}, {linkId=-1, sourceLinkId=-1}], blendTime [0.1,0.1]}
   n91 = AimOffset Player 'M_AO_AR_Downsights'(X=0, Y=@n43, BasePose=@n90, Alpha=1) {blendSpace=&M_AO_AR_Downsights, alphaBoolBlend {blendOption=Linear}}
   n92 = "Blend Poses (EFortCustomGender)"(
    BlendPose_0=Sequence Player 'Shotgun_StraightGrip_Relaxed_CMM'() {sequence=/Game/Animation/Game/MainPlayer/Poses/Medium/Male/Shotgun_StraightGrip_Relaxed_CMM}
    BlendPose_1=Sequence Player 'AR_Relaxed_Pose_CMF'() {sequence=&AR_Relaxed_Pose_CMF}
    BlendTime_0=0.1
    BlendTime_1=0.1
    ActiveEnumValue=Get Gender() -> (Gender=Invalid)
   ) {blendType=Linear, blendPose [{linkId=-1, sourceLinkId=-1}, {linkId=-1, sourceLinkId=-1}], blendTime [0.1,0.1]}
   n93 = AimOffset Player 'M_NoWep_Relaxed_AO'(X=0, Y=@n43, BasePose=@n92, Alpha=1) {blendSpace=&M_NoWep_Relaxed_AO, alphaBoolBlend {blendOption=Linear}}
   n94 = "Blend Poses (EFortCustomGender)"(
    BlendPose_0=Sequence Player 'Shotgun_StraightGrip_NonTargeted_CMM'() {sequence=/Game/Animation/Game/MainPlayer/Poses/Medium/Male/Shotgun_StraightGrip_NonTargeted_CMM}
    BlendPose_1=Sequence Player 'AR_NonTargeted_Pose_CMF'() {sequence=&AR_NonTargeted_Pose_CMF}
    BlendTime_0=0.1
    BlendTime_1=0.1
    ActiveEnumValue=Get Gender() -> (Gender=Invalid)
   ) {blendType=Linear, blendPose [{linkId=-1, sourceLinkId=-1}, {linkId=-1, sourceLinkId=-1}], blendTime [0.1,0.1]}
   n95 = AimOffset Player 'M_AO_AR_Downsights'(X=0, Y=@n43, BasePose=@n94, Alpha=1) {blendSpace=&M_AO_AR_Downsights, alphaBoolBlend {blendOption=Linear}}
   n96 = OR Boolean(A=Get bIsRelaxedLevel1(), B=NOT Boolean(A=Get bIsTargeting()))
   n97 = "Blend Poses (EFortCustomGender)"(
    BlendPose_0=Sequence Player 'Shotgun_StraightGrip_Targeted_CMM'() {sequence=/Game/Animation/Game/MainPlayer/Poses/Medium/Male/Shotgun_StraightGrip_Targeted_CMM}
    BlendPose_1=Sequence Player 'AR_Targeted_Pose_CMF'() {sequence=&AR_Targeted_Pose_CMF}
    BlendTime_0=0.1
    BlendTime_1=0.1
    ActiveEnumValue=Get Gender() -> (Gender=Invalid)
   ) {blendType=Linear, blendPose [{linkId=-1, sourceLinkId=-1}, {linkId=-1, sourceLinkId=-1}], blendTime [0.1,0.1]}
   n98 = AimOffset Player 'M_AO_AR_Downsights'(X=0, Y=@n43, BasePose=@n97, Alpha=1) {blendSpace=&M_AO_AR_Downsights, alphaBoolBlend {blendOption=Linear}}
   n99 = "Blend Poses (EFortCustomGender)"(
    BlendPose_0=Sequence Player 'Pose_ARBullPup_Relaxed_CMM'() {sequence=/Game/Animation/Game/MainPlayer/Poses/Medium/Male/BullPup/Pose_ARBullPup_Relaxed_CMM}
    BlendPose_1=Sequence Player 'Pose_AR_BullPup_Relaxed_CMF'() {sequence=/Game/Animation/Game/MainPlayer/Poses/Medium/Female/BullPup/Pose_AR_BullPup_Relaxed_CMF}
    BlendTime_0=0.1
    BlendTime_1=0.1
    ActiveEnumValue=Get Gender() -> (Gender=Invalid)
   ) {blendType=Linear, blendPose [{linkId=-1, sourceLinkId=-1}, {linkId=-1, sourceLinkId=-1}], blendTime [0.1,0.1]}
   n100 = float * float(A=float - float(B=GetCurveValue(CurveName=ZeroOutPitchWeight)), B=Get AimPitch())
   n101 = AimOffset Player 'M_NoWep_Relaxed_AO'(X=0, Y=@n100, BasePose=@n99, Alpha=1) {blendSpace=&M_NoWep_Relaxed_AO, alphaBoolBlend {blendOption=Linear}}
   n102 = "Blend Poses (EFortCustomGender)"(
    BlendPose_0=Sequence Player 'Pose_ARBullPup_NonTargeted_CMM'() {sequence=/Game/Animation/Game/MainPlayer/Poses/Medium/Male/BullPup/Pose_ARBullPup_NonTargeted_CMM}
    BlendPose_1=Sequence Player 'Pose_AR_BullPup_NonTargeted_CMF'() {sequence=/Game/Animation/Game/MainPlayer/Poses/Medium/Female/BullPup/Pose_AR_BullPup_NonTargeted_CMF}
    BlendTime_0=0.1
    BlendTime_1=0.1
    ActiveEnumValue=Get Gender() -> (Gender=Invalid)
   ) {blendType=Linear, blendPose [{linkId=-1, sourceLinkId=-1}, {linkId=-1, sourceLinkId=-1}], blendTime [0.1,0.1]}
   n103 = AimOffset Player 'M_AO_AR_Downsights'(X=0, Y=@n100, BasePose=@n102, Alpha=1) {blendSpace=&M_AO_AR_Downsights, alphaBoolBlend {blendOption=Linear}}
   n104 = OR Boolean(A=Get bIsRelaxedLevel1(), B=NOT Boolean(A=Get bIsTargeting()))
   n105 = "Blend Poses (EFortCustomGender)"(
    BlendPose_0=Sequence Player 'Pose_ARBullPup_Targeted_CMM'() {sequence=/Game/Animation/Game/MainPlayer/Poses/Medium/Male/BullPup/Pose_ARBullPup_Targeted_CMM}
    BlendPose_1=Sequence Player 'Pose_AR_BullPup_Targeted_CMF'() {sequence=/Game/Animation/Game/MainPlayer/Poses/Medium/Female/BullPup/Pose_AR_BullPup_Targeted_CMF}
    BlendTime_0=0.1
    BlendTime_1=0.1
    ActiveEnumValue=Get Gender() -> (Gender=Invalid)
   ) {blendType=Linear, blendPose [{linkId=-1, sourceLinkId=-1}, {linkId=-1, sourceLinkId=-1}], blendTime [0.1,0.1]}
   n106 = AimOffset Player 'M_AO_AR_Downsights'(X=0, Y=@n100, BasePose=@n105, Alpha=1) {blendSpace=&M_AO_AR_Downsights, alphaBoolBlend {blendOption=Linear}}
   n107 = "Blend Poses (EFortCustomGender)"(
    BlendPose_0=Sequence Player 'M_NoWep_Relaxed'() {sequence=&M_NoWep_Relaxed}
    BlendPose_1=Sequence Player 'Pose_NoWep_Relaxed_CMF'() {sequence=&Pose_NoWep_Relaxed_CMF}
    BlendTime_0=00
    BlendTime_1=0
    ActiveEnumValue=Get Gender() -> (Gender=Invalid)
   ) {blendType=Linear, blendPose [{linkId=-1, sourceLinkId=-1}, {linkId=-1, sourceLinkId=-1}], blendTime [0.1,0.1]}
   n108 = "Blend Poses (EFortCustomGender)"(
    BlendPose_0=Sequence Player 'M_Melee_OneHanded'() {sequence=/Game/Animation/Game/MainPlayer/Poses/Medium/Male/M_Melee_OneHanded}
    BlendPose_1=Sequence Player 'Pose_Melee_OneHanded_CMF'() {sequence=/Game/Animation/Game/MainPlayer/Poses/Medium/Female/Pose_Melee_OneHanded_CMF}
    BlendTime_0=0.1
    BlendTime_1=0.1
    ActiveEnumValue=Get Gender() -> (Gender=Invalid)
   ) {blendType=Linear, blendPose [{linkId=-1, sourceLinkId=-1}, {linkId=-1, sourceLinkId=-1}], blendTime [0.1,0.1]}
   n109 = float * float(A=float - float(B=GetCurveValue(CurveName=ZeroOutPitchWeight)), B=Get AimPitch())
   n110 = AimOffset Player 'M_AO_AR_Downsights'(X=0, Y=@n109, BasePose=Sequence Player 'M_RPG_DS'() {sequence=/Game/Animation/Game/MainPlayer/Poses/Medium/Male/M_RPG_DS}, Alpha=1) {blendSpace=&M_AO_AR_Downsights, alphaBoolBlend {blendOption=Linear}}
   n111 = "Blend Poses (EFortCustomGender)"(
    BlendPose_0=Sequence Player 'Spear_NonTargeted_Pose_CMM'() {sequence=/Game/Animation/Game/MainPlayer/Poses/Medium/Male/Spear/Spear_NonTargeted_Pose_CMM}
    BlendPose_1=Sequence Player 'Spear_NonTargeted_Pose_CMM'() {sequence=&Spear_NonTargeted_Pose_CMM}
    BlendTime_0=0.1
    BlendTime_1=0.1
    ActiveEnumValue=Get Gender() -> (Gender=Invalid)
   ) {blendType=Linear, blendPose [{linkId=-1, sourceLinkId=-1}, {linkId=-1, sourceLinkId=-1}], blendTime [0.1,0.1]}
   n112 = AimOffset Player 'M_NoWep_Relaxed_AO'(X=0, Y=@n109, BasePose=@n111, Alpha=1) {blendSpace=&M_NoWep_Relaxed_AO, alphaBoolBlend {blendOption=Linear}}
   n113 = "Blend Poses (EFortCustomGender)"(
    BlendPose_0=Sequence Player 'M_Melee_TwoHanded'() {sequence=/Game/Animation/Game/MainPlayer/Poses/Medium/Male/M_Melee_TwoHanded}
    BlendPose_1=Sequence Player 'Pose_Melee_TwoHanded_CMF'() {sequence=/Game/Animation/Game/MainPlayer/Poses/Medium/Female/Pose_Melee_TwoHanded_CMF}
    BlendTime_0=0.1
    BlendTime_1=0.1
    ActiveEnumValue=Get Gender() -> (Gender=Invalid)
   ) {blendType=Linear, blendPose [{linkId=-1, sourceLinkId=-1}, {linkId=-1, sourceLinkId=-1}], blendTime [0.1,0.1]}
   n114 = AimOffset Player 'M_NoWep_Relaxed_AO'(X=0, Y=@n109, BasePose=@n113, Alpha=1) {blendSpace=&M_NoWep_Relaxed_AO, alphaBoolBlend {blendOption=Linear}}
   n115 = Get UsingSpearWithThrust() -> (UsingSpearWithThrust:bool)
   n116 = "Blend Poses (EFortCustomGender)"(
    BlendPose_0=Sequence Player 'AthenaBlade_NonTargeted_Pose_CMM'() {sequence=/Game/Animation/Game/MainPlayer/Poses/Medium/Male/AthenaBlade/AthenaBlade_NonTargeted_Pose_CMM}
    BlendPose_1=Sequence Player 'AthenaBlade_NonTargeted_Pose_CMM'() {sequence=&AthenaBlade_NonTargeted_Pose_CMM}
    BlendTime_0=0.1
    BlendTime_1=0.1
    ActiveEnumValue=Get Gender() -> (Gender=Invalid)
   ) {blendType=Linear, blendPose [{linkId=-1, sourceLinkId=-1}, {linkId=-1, sourceLinkId=-1}], blendTime [0.1,0.1]}
   n117 = "Blend Poses (EFortWeaponCoreAnimation)"!(
    BlendPose_0:Pose Link Structure=@n107
    BlendPose_1:Pose Link Structure=Blend Poses by bool(
     BlendPose_0=AimOffset Player 'M_AO_Pistol_Downsights'(X=@n50, Y=@n43, BasePose=@n49, Alpha=1) {blendSpace=&M_AO_Pistol_Downsights, alphaBoolBlend {blendOption=Linear}}
     BlendPose_1=Blend Poses by bool(BlendPose_0=@n47, BlendPose_1=@n44, BlendTime_0=.4, BlendTime_1=0, bActiveValue=@n48) {blendType=Linear}
     BlendTime_1=.4
     bActiveValue=Get bIsTargeting()
    ) {blendType=Linear}
    BlendPose_2:Pose Link Structure=Blend Poses by bool(
     BlendPose_0=@n91
     BlendPose_1=Blend Poses by bool(BlendPose_0=@n86, BlendPose_1=@n88, BlendTime_0=0.4, BlendTime_1=0, bActiveValue=@n89) {blendType=Linear}
     BlendTime_1=0.42
     bActiveValue=Get bIsTargeting()
    ) {blendType=Linear}
    BlendPose_3:Pose Link Structure=@n110
    BlendPose_4:Pose Link Structure=Blend Poses by bool(
     BlendPose_0=AimOffset Player 'M_NoWep_Relaxed_AO'(X=0, Y=@n109, BasePose=@n116, Alpha=1) {blendSpace=&M_NoWep_Relaxed_AO, alphaBoolBlend {blendOption=Linear}}
     BlendPose_1=Blend Poses by bool(BlendPose_0=@n112, BlendPose_1=@n114, bActiveValue=@n115) {blendType=Linear}
     bActiveValue=Get Is Using Two Handed Sword() -> (Is Using Two Handed Sword:bool)
    ) {blendType=Linear}
    BlendPose_5:Pose Link Structure=AimOffset Player 'M_NoWep_Relaxed_AO'(X=0, Y=@n109, BasePose=@n108, Alpha=1) {blendSpace=&M_NoWep_Relaxed_AO, alphaBoolBlend {blendOption=Linear}}
    BlendPose_6:Pose Link Structure=AimOffset Player 'M_GoingCommando_AO'!(
     X:float=0
     Y:float=@n109
     BasePose:Pose Link Structure=Sequence Player 'M_GoingCommando'() {sequence=/Game/Animation/Game/MainPlayer/Poses/Medium/Male/M_GoingCommando}
     Alpha:float=1
    ) -> (Pose:Pose Link Structure) {blendSpace=/Game/Animation/Game/MainPlayer/Poses/Medium/Male/AO/M_GoingCommando_AO, alphaBoolBlend {blendOption=Linear}}
    BlendPose_7:Pose Link Structure=Blend Poses by bool(
     BlendPose_0=@n57
     BlendPose_1=Blend Poses by bool(BlendPose_0=@n52, BlendPose_1=@n54, BlendTime_0=0.4, BlendTime_1=0, bActiveValue=@n55) {blendType=Linear}
     BlendTime_1=0.42
     bActiveValue=Get bIsTargeting()
    ) {blendType=Linear}
    BlendPose_8:Pose Link Structure=Blend Poses by bool(
     BlendPose_0=@n41
     BlendPose_1=Blend Poses by bool(BlendPose_0=@n36, BlendPose_1=@n38, BlendTime_0=0.4, BlendTime_1=0, bActiveValue=@n39) {blendType=Linear}
     BlendTime_1=0.42
     bActiveValue=Get bIsTargeting()
    ) {blendType=Linear}
    BlendPose_9:Pose Link Structure=Blend Poses by bool(
     BlendPose_0=@n98
     BlendPose_1=Blend Poses by bool(BlendPose_0=@n93, BlendPose_1=@n95, BlendTime_0=0.4, BlendTime_1=0, bActiveValue=@n96) {blendType=Linear}
     BlendTime_1=0.42
     bActiveValue=Get bIsTargeting()
    ) {blendType=Linear}
    BlendPose_10:Pose Link Structure=Blend Poses by bool(
     BlendPose_0=@n17
     BlendPose_1=Blend Poses by bool(BlendPose_0=@n12, BlendPose_1=@n14, BlendTime_0=0.4, BlendTime_1=0, bActiveValue=@n15) {blendType=Linear}
     BlendTime_1=0.42
     bActiveValue=Get bIsTargeting()
    ) {blendType=Linear}
    BlendPose_11:Pose Link Structure=Blend Poses by bool(
     BlendPose_0=@n106
     BlendPose_1=Blend Poses by bool(BlendPose_0=@n101, BlendPose_1=@n103, BlendTime_0=0.4, BlendTime_1=0, bActiveValue=@n104) {blendType=Linear}
     BlendTime_1=0.42
     bActiveValue=Get bIsTargeting()
    ) {blendType=Linear}
    BlendPose_12:Pose Link Structure=Blend Poses by bool(
     BlendPose_0=@n76
     BlendPose_1=Blend Poses by bool(BlendPose_0=@n71, BlendPose_1=@n73, BlendTime_0=0.4, BlendTime_1=0, bActiveValue=@n74) {blendType=Linear}
     BlendTime_1=0.42
     bActiveValue=Get bIsTargeting()
    ) {blendType=Linear}
    BlendPose_13:Pose Link Structure=Blend Poses by bool(
     BlendPose_0=@n84
     BlendPose_1=Blend Poses by bool(BlendPose_0=@n79, BlendPose_1=@n81, BlendTime_0=0.4, BlendTime_1=0, bActiveValue=@n82) {blendType=Linear}
     BlendTime_1=0.42
     bActiveValue=Get bIsTargeting()
    ) {blendType=Linear}
    BlendPose_14:Pose Link Structure=Blend Poses by bool(
     BlendPose_0=@n25
     BlendPose_1=Blend Poses by bool(BlendPose_0=@n20, BlendPose_1=@n22, BlendTime_0=0.4, BlendTime_1=0, bActiveValue=@n23) {blendType=Linear}
     BlendTime_1=0.42
     bActiveValue=Get bIsTargeting()
    ) {blendType=Linear}
    BlendPose_15:Pose Link Structure=Blend Poses by bool(
     BlendPose_0=@n33
     BlendPose_1=Blend Poses by bool(BlendPose_0=@n28, BlendPose_1=@n30, BlendTime_0=0.4, BlendTime_1=0, bActiveValue=@n31) {blendType=Linear}
     BlendTime_1=0.42
     bActiveValue=Get bIsTargeting()
    ) {blendType=Linear}
    BlendPose_16:Pose Link Structure=Blend Poses by bool(
     BlendPose_0=AimOffset Player 'M_NoWep_Relaxed_AO'(X=0, Y=@n66, BasePose=@n65, Alpha=1) {blendSpace=&M_NoWep_Relaxed_AO, alphaBoolBlend {blendOption=Linear}}
     BlendPose_1=Blend Poses by bool(BlendPose_0=@n61, BlendPose_1=@n63, BlendTime_0=0.55, BlendTime_1=0, bActiveValue=@n64) {blendType=Linear}
     BlendTime_0=0.4
     bActiveValue=OR Boolean(A=@n68, B=@n67)
    ) {blendType=Linear}
    BlendTime_0:float=0
    BlendTime_1:float=0
    BlendTime_2:float=0
    BlendTime_3:float=0
    BlendTime_4:float=0
    BlendTime_5:float=0
    BlendTime_6:float=0.1
    BlendTime_7:float=0.1
    BlendTime_8:float=0.1
    BlendTime_9:float=0.1
    BlendTime_10:float=0.1
    BlendTime_11:float=0.1
    BlendTime_12:float=0.1
    BlendTime_13:float=0.1
    BlendTime_14:float=0.1
    BlendTime_15:float=0.1
    BlendTime_16:float=0.1
    ActiveEnumValue:Byte=Get WeaponCoreAnimation() -> (WeaponCoreAnimation:EFortWeaponCoreAnimation Enum)
   ) -> (Pose:Pose Link Structure) {
    is=title AnimGraphNode_BlendListByEnum
    blendType=Linear
    blendPose [
     {linkId=-1, sourceLinkId=-1}
     {linkId=-1, sourceLinkId=-1}
     {linkId=-1, sourceLinkId=-1}
     {linkId=-1, sourceLinkId=-1}
     {linkId=-1, sourceLinkId=-1}
     {linkId=-1, sourceLinkId=-1}
     {linkId=-1, sourceLinkId=-1}
     {linkId=-1, sourceLinkId=-1}
     {linkId=-1, sourceLinkId=-1}
     {linkId=-1, sourceLinkId=-1}
     {linkId=-1, sourceLinkId=-1}
     {linkId=-1, sourceLinkId=-1}
     {linkId=-1, sourceLinkId=-1}
     {linkId=-1, sourceLinkId=-1}
     {linkId=-1, sourceLinkId=-1}
     {linkId=-1, sourceLinkId=-1}
     {linkId=-1, sourceLinkId=-1}
    ]
    blendTime [0.1,0.1,0.1,0.1,0.1,0.1,0.1,0.1,0.1,0.1,0.1,0.1,0.1,0.1,0.1,0.1,0.1]
   }
   n118 = Get bHasWeapon()
   n119 = float * float(A=float - float(B=GetCurveValue(CurveName=ZeroOutPitchWeight)), B=Get AimPitch())
   n120 = Slot 'MeshRotationUpperBody'(Source=Rotate Root Bone(
     BasePose=Blend Poses by bool(BlendPose_0=@n117, BlendPose_1=@n9, bActiveValue=@n118) {blendType=Linear}
     Pitch=float * float!(A:double=@n119, B:double=-1) -> (ReturnValue:double) {is=title K2Node_CallFunction}
     Yaw=0
     MeshToComponent="0,-90,0"
    )) {slotName=MeshRotationUpperBody}
   Save cached pose 'UpperBody'(Pose=Slot 'UpperBody'(Source=Rotate Root Bone(BasePose=@n120, Pitch=@n119, Yaw=0, MeshToComponent="0,-90,0")) {slotName=UpperBody}) {cacheName=UpperBody}
   AnimGraphNode_StateMachine_56 = LocomotionLowerBody() {is=title AnimGraphNode_StateMachine}
   n121 = Layered blend per bone(BasePose=Use cached pose 'UpperBody'(), BlendPoses_0=@AnimGraphNode_StateMachine_56) {
    bMeshSpaceRotationBlend=true
    curveBlendOption=BlendByWeight
    layerSetup [{branchFilters [{boneName=Root}, {blendDepth=-1, boneName=Ik_hand_root}, {blendDepth=-2, boneName=Spine_01}]}]
   }
   Save cached pose 'UpperAndLowerbody'(Pose=@n121) {cacheName=UpperAndLowerbody}
   n122 = Cast To FortPawn!(Object=GetOwningActor()) -> (AsFort Pawn:Fort Pawn Object Reference, bSuccess:bool=~) {is=title K2Node_DynamicCast}
   "Equal (Object)"(
    A=GetWeaponPawnAnimSet!(self:Fort Weapon Object Reference=Get CurrentWeapon!(self:Fort Pawn Object Reference=@n122) -> (CurrentWeapon:Fort Weapon Object Reference)) -> (ReturnValue:Fort Weapon Anim Set Object Reference)
    B=/Game/Animation/Game/MainPlayer/AnimAssets/Spear_AnimAsset
   ) {is=title K2Node_CallFunction}
   Get AimYaw()
   Get AimYaw()
   Get AimYaw()
   Get AimYaw()
   Get AimYaw()
   Get AimYaw()
   Get AimYaw()
   Get WeaponCoreAnimation()
   Get AimYaw()
  ]}
 SurvivorAdditive {
  entry=Idle
  path=AnimGraph.AnimGraphNode_StateMachine_38.SurvivorAdditive
  ids=id
  nodes [
   AnimStateNode_4 = Idle()
   AnimStateNode_6 = HitReactLoop() {is=title AnimStateNode}
   AnimStateTransitionNode_1 = AnimStateTransitionNode()
   AnimStateTransitionNode_3 = AnimStateTransitionNode()
  ]
  links [
   {
    dst=Idle
    rule=AnimGraph.AnimGraphNode_StateMachine_38.SurvivorAdditive.AnimStateTransitionNode_3.Transition
    src=HitReactLoop
    via=AnimStateTransitionNode_3
   }
   {
    dst=HitReactLoop
    rule=AnimGraph.AnimGraphNode_StateMachine_38.SurvivorAdditive.AnimStateTransitionNode_1.Transition
    src=Idle
    via=AnimStateTransitionNode_1
   }
  ]
 }
 Idle {
  path=AnimGraph.AnimGraphNode_StateMachine_38.SurvivorAdditive.AnimStateNode_4.Idle
  nodes [Output Animation Pose(Result=Additive Identity Pose())]
 }
 HitReactLoop {
  path=AnimGraph.AnimGraphNode_StateMachine_38.SurvivorAdditive.AnimStateNode_6.HitReactLoop
  nodes [Output Animation Pose(Result="Sequence Player (additive) 'FromStanding_InFront_CMM'"() {sequence=/Game/Animation/Game/MainPlayer/Reacts/HitReacts/FromStanding_InFront_CMM})]
 }
 Transition {
  path=AnimGraph.AnimGraphNode_StateMachine_38.SurvivorAdditive.AnimStateTransitionNode_1.Transition
  nodes [Result(bCanEnterTransition=Get bAdditiveHitReactLoop() -> (bAdditiveHitReactLoop:bool))]
 }
 Transition {
  path=AnimGraph.AnimGraphNode_StateMachine_38.SurvivorAdditive.AnimStateTransitionNode_3.Transition
  nodes [Result(bCanEnterTransition=NOT Boolean(A=Get bAdditiveHitReactLoop()))]
 }
 SurvivorFullBody {
  entry=SplitBody
  path=AnimGraph.AnimGraphNode_StateMachine_55.SurvivorFullBody
  ids=id
  nodes [
   AnimStateNode_10 = Stunned() {is=title AnimStateNode}
   AnimStateNode_2 = SplitBody() {is=title AnimStateNode}
   AnimStateNode_3 = DBNO() {is=title AnimStateNode}
   AnimStateNode_9 = Cower() {is=title AnimStateNode}
   AnimStateTransitionNode_11 = AnimStateTransitionNode()
   AnimStateTransitionNode_12 = AnimStateTransitionNode()
   AnimStateTransitionNode_13 = AnimStateTransitionNode()
   AnimStateTransitionNode_14 = AnimStateTransitionNode()
   AnimStateTransitionNode_2 = AnimStateTransitionNode()
   AnimStateTransitionNode_20 = AnimStateTransitionNode()
   AnimStateTransitionNode_5 = AnimStateTransitionNode()
   AnimStateTransitionNode_7 = AnimStateTransitionNode()
  ]
  links [
   {
    dst=DBNO
    rule=AnimGraph.AnimGraphNode_StateMachine_55.SurvivorFullBody.AnimStateTransitionNode_13.Transition
    src=Cower
    via=AnimStateTransitionNode_13
   }
   {
    dst=SplitBody
    rule=AnimGraph.AnimGraphNode_StateMachine_55.SurvivorFullBody.AnimStateTransitionNode_7.Transition
    src=Cower
    via=AnimStateTransitionNode_7
   }
   {
    dst=Cower
    rule=AnimGraph.AnimGraphNode_StateMachine_55.SurvivorFullBody.AnimStateTransitionNode_11.Transition
    src=SplitBody
    via=AnimStateTransitionNode_5
   }
   {
    dst=DBNO
    rule=AnimGraph.AnimGraphNode_StateMachine_55.SurvivorFullBody.AnimStateTransitionNode_13.Transition
    src=SplitBody
    via=AnimStateTransitionNode_2
   }
   {
    dst=Stunned
    rule=AnimGraph.AnimGraphNode_StateMachine_55.SurvivorFullBody.AnimStateTransitionNode_12.Transition
    src=SplitBody
    via=AnimStateTransitionNode_12
   }
   {
    dst=Cower
    rule=AnimGraph.AnimGraphNode_StateMachine_55.SurvivorFullBody.AnimStateTransitionNode_11.Transition
    src=Stunned
    via=AnimStateTransitionNode_11
   }
   {
    dst=DBNO
    rule=AnimGraph.AnimGraphNode_StateMachine_55.SurvivorFullBody.AnimStateTransitionNode_13.Transition
    src=Stunned
    via=AnimStateTransitionNode_20
   }
   {
    dst=SplitBody
    rule=AnimGraph.AnimGraphNode_StateMachine_55.SurvivorFullBody.AnimStateTransitionNode_14.Transition
    src=Stunned
    via=AnimStateTransitionNode_14
   }
  ]
 }
 Stunned {
  path=AnimGraph.AnimGraphNode_StateMachine_55.SurvivorFullBody.AnimStateNode_10.Stunned
  nodes [Output Animation Pose(Result=Sequence Player 'Standing_Stun'() {sequence=/Game/Animation/Game/MainPlayer/Locomotion/Small/idle/Standing_Stun})]
 }
 SplitBody {
  path=AnimGraph.AnimGraphNode_StateMachine_55.SurvivorFullBody.AnimStateNode_2.SplitBody
  nodes [
   Apply Additive!(
    Base=Sequence Player 'M_NoWep_Relaxed'() {sequence=&M_NoWep_Relaxed}
    Additive="Sequence Player (additive) 'Idle_Noise'"() {sequence=/Game/Animation/Game/MainPlayer/Locomotion/Medium/Male/Idle/Idle_Noise}
    Alpha:float
   ) -> (Pose) {alphaBoolBlend {blendOption=Linear}}
   AnimGraphNode_StateMachine_87 = LocomotionAdditive() {is=title AnimGraphNode_StateMachine}
   Output Animation Pose(Result=Apply Additive(Base=Use cached pose 'UpperAndLowerbody'(), Additive=@AnimGraphNode_StateMachine_87) {alphaBoolBlend {blendOption=Linear}})
  ]
 }
 LocomotionAdditive {
  entry=IdleAdditive
  path=AnimGraph.AnimGraphNode_StateMachine_55.SurvivorFullBody.AnimStateNode_2.SplitBody.AnimGraphNode_StateMachine_87.LocomotionAdditive
  ids=id
  nodes [
   AnimStateNode_2 = IdleAdditive() {is=title AnimStateNode}
   AnimStateNode_3 = MovementAdditive() {is=title AnimStateNode}
   AnimStateTransitionNode_2 = AnimStateTransitionNode()
   AnimStateTransitionNode_3 = AnimStateTransitionNode()
  ]
  links [
   {
    dst=MovementAdditive
    rule=AnimGraph.AnimGraphNode_StateMachine_55.SurvivorFullBody.AnimStateNode_2.SplitBody.AnimGraphNode_StateMachine_87.LocomotionAdditive.AnimStateTransitionNode_2.Transition
    src=IdleAdditive
    via=AnimStateTransitionNode_2
   }
   {
    dst=IdleAdditive
    rule=AnimGraph.AnimGraphNode_StateMachine_55.SurvivorFullBody.AnimStateNode_2.SplitBody.AnimGraphNode_StateMachine_87.LocomotionAdditive.AnimStateTransitionNode_3.Transition
    src=MovementAdditive
    via=AnimStateTransitionNode_3
   }
  ]
 }
 IdleAdditive {
  path=AnimGraph.AnimGraphNode_StateMachine_55.SurvivorFullBody.AnimStateNode_2.SplitBody.AnimGraphNode_StateMachine_87.LocomotionAdditive.AnimStateNode_2.IdleAdditive
  nodes [Output Animation Pose(Result=Blend Poses by bool(
     BlendPose_0="Sequence Player (additive) 'Idle_Noise_AR_Downsights'"() {sequence=/Game/Animation/Game/MainPlayer/Locomotion/Medium/Male/Idle/Idle_Noise_AR_Downsights}
     BlendPose_1="Sequence Player (additive) 'Idle_Noise'"() {sequence=&Idle_Noise}
     bActiveValue=Get bHasWeapon()
    ) {blendType=Linear})]
 }
 MovementAdditive {
  path=AnimGraph.AnimGraphNode_StateMachine_55.SurvivorFullBody.AnimStateNode_2.SplitBody.AnimGraphNode_StateMachine_87.LocomotionAdditive.AnimStateNode_3.MovementAdditive
  nodes [
   n1 = Get Speed() -> (Speed:float)
   n2 = Get MovementDirection() -> (MovementDirection:float)
   n3 = Blendspace Player 'M_Jog_ARRelaxed_Additive'(X=@n2, Y=@n1) {
    blendSpace=/Game/Animation/Game/MainPlayer/Locomotion/Medium/Male/Jog/AssaultRifleRelaxed/M_Jog_ARRelaxed_Additive
    groupName=Locomotion
    groupRole=AlwaysFollower
    method=SyncGroup
   }
   Blend Poses by bool(BlendPose_0=@n3, BlendTime_0=.4, BlendTime_1=0, bActiveValue=Get bIsRelaxedLevel1()) {blendType=Linear}
   n4 = Get Speed()
   n5 = Blendspace Player 'AssaultRifle_Relaxed_CMF_AD'(X=@n2, Y=@n4) {
    blendSpace=/Game/Animation/Game/MainPlayer/Locomotion/Medium/Female/Jog/AssaultRifle_Relaxed/AssaultRifle_Relaxed_CMF_AD
    groupName=Locomotion
    groupRole=AlwaysFollower
    method=SyncGroup
   }
   Blend Poses by bool(BlendPose_0=@n5, BlendTime_0=.4, BlendTime_1=0, bActiveValue=Get bIsRelaxedLevel1()) {blendType=Linear}
   n6 = Get Speed()
   n7 = Get MovementDirection()
   n8 = Blendspace Player 'NoWep_CMF_ADD'(X=@n7, Y=@n6) {
    blendSpace=/Game/Animation/Game/MainPlayer/Locomotion/Medium/Female/Jog/NoWeapon/NoWep_CMF_ADD
    groupName=Locomotion
    groupRole=AlwaysFollower
    method=SyncGroup
   }
   n9 = Blendspace Player 'M_Jog_NoWep_Additive'(X=@n7, Y=@n6) {
    blendSpace=/Game/Animation/Game/MainPlayer/Locomotion/Medium/Male/Jog/NoWeapon/M_Jog_NoWep_Additive
    groupName=Locomotion
    groupRole=AlwaysFollower
    method=SyncGroup
   }
   n10 = "Blend Poses (EFortCustomGender)"(BlendPose_0=@n9, BlendPose_1=@n8, BlendTime_0=0.1, BlendTime_1=0.1, ActiveEnumValue=Get Gender() -> (Gender=Invalid)) {blendType=Linear, blendPose [{linkId=-1, sourceLinkId=-1}, {linkId=-1, sourceLinkId=-1}], blendTime [0.1,0.1]}
   n11 = "Blend Poses (EFortMovementStyle)"!(
    BlendPose_0:Pose Link Structure=@n10
    BlendPose_1:Pose Link Structure=Additive Identity Pose()
    BlendTime_0:float=0.1
    BlendTime_1:float=0.1
    ActiveEnumValue:Byte=Get MovementStyle() -> (MovementStyle:EFortMovementStyle Enum=Running)
   ) -> (Pose:Pose Link Structure) {
    is=title AnimGraphNode_BlendListByEnum
    blendType=Linear
    blendPose [{linkId=-1, sourceLinkId=-1}, {linkId=-1, sourceLinkId=-1}]
    blendTime [0.1,0.1]
   }
   n12 = Get Speed()
   n13 = Blendspace Player 'M_Jog_Melee_OH_Additive'(X=@n2, Y=@n12) {
    blendSpace=/Game/Animation/Game/MainPlayer/Locomotion/Medium/Male/Jog/Melee_OH/M_Jog_Melee_OH_Additive
    groupName=Locomotion
    groupRole=AlwaysFollower
    method=SyncGroup
   }
   n14 = Blendspace Player 'Melee_OneHanded_CMF_ADD'(X=@n2, Y=@n12) {
    blendSpace=/Game/Animation/Game/MainPlayer/Locomotion/Medium/Female/Jog/Melee_OneHanded/Melee_OneHanded_CMF_ADD
    groupName=Locomotion
    groupRole=AlwaysFollower
    method=SyncGroup
   }
   n15 = "Blend Poses (EFortCustomGender)"(BlendPose_0=@n13, BlendPose_1=@n14, BlendTime_0=0.1, BlendTime_1=0.1, ActiveEnumValue=Get Gender() -> (Gender=Invalid)) {blendType=Linear, blendPose [{linkId=-1, sourceLinkId=-1}, {linkId=-1, sourceLinkId=-1}], blendTime [0.1,0.1]}
   n16 = Blendspace Player 'M_Jog_ARRelaxed_Additive'(X=@n2, Y=@n12) {
    blendSpace=&M_Jog_ARRelaxed_Additive
    groupName=Locomotion
    groupRole=AlwaysFollower
    method=SyncGroup
   }
   n17 = Blendspace Player 'AssaultRifle_Relaxed_CMF_AD'(X=@n2, Y=@n12) {
    blendSpace=&AssaultRifle_Relaxed_CMF_AD
    groupName=Locomotion
    groupRole=AlwaysFollower
    method=SyncGroup
   }
   n18 = "Blend Poses (EFortCustomGender)"(BlendPose_0=@n16, BlendPose_1=@n17, BlendTime_0=0.1, BlendTime_1=0.1, ActiveEnumValue=Get Gender() -> (Gender=Invalid)) {blendType=Linear, blendPose [{linkId=-1, sourceLinkId=-1}, {linkId=-1, sourceLinkId=-1}], blendTime [0.1,0.1]}
   n19 = Get bIsRelaxedLevel2()
   n20 = Get Speed()
   n21 = Blendspace Player 'M_Jog_PistolHF_Additive'(X=@n2, Y=@n20) {
    blendSpace=/Game/Animation/Game/MainPlayer/Locomotion/Medium/Male/Jog/PistolHF/M_Jog_PistolHF_Additive
    groupName=Locomotion
    groupRole=AlwaysFollower
    method=SyncGroup
   }
   n22 = Blendspace Player 'Pistol_NonTargeted_CMF_ADD'(X=@n2, Y=@n20) {
    blendSpace=/Game/Animation/Game/MainPlayer/Locomotion/Medium/Female/Jog/Pistol_NonTargeted/Pistol_NonTargeted_CMF_ADD
    groupName=Locomotion
    groupRole=AlwaysFollower
    method=SyncGroup
   }
   n23 = "Blend Poses (EFortCustomGender)"(BlendPose_0=@n21, BlendPose_1=@n22, BlendTime_0=0.1, BlendTime_1=0.1, ActiveEnumValue=Get Gender() -> (Gender=Invalid)) {blendType=Linear, blendPose [{linkId=-1, sourceLinkId=-1}, {linkId=-1, sourceLinkId=-1}], blendTime [0.1,0.1]}
   n24 = Get Speed()
   n25 = Blendspace Player 'M_Jog_ARRelaxed_Additive'(X=@n2, Y=@n24) {
    blendSpace=&M_Jog_ARRelaxed_Additive
    groupName=Locomotion
    groupRole=AlwaysFollower
    method=SyncGroup
   }
   n26 = Blendspace Player 'M_Jog_AR_Additive'(X=@n2, Y=@n24) {
    blendSpace=/Game/Animation/Game/MainPlayer/Locomotion/Medium/Male/Jog/AssaultRifle_HipNonTargeted/M_Jog_AR_Additive
    groupName=Locomotion
    groupRole=AlwaysFollower
    method=SyncGroup
   }
   n27 = Get Speed()
   n28 = Blendspace Player 'AssaultRifle_HFNT_CMF_Add'(X=@n2, Y=@n27) {
    blendSpace=/Game/Animation/Game/MainPlayer/Locomotion/Medium/Female/Jog/AssaultRifle_HFNonTargeted/AssaultRifle_HFNT_CMF_Add
    groupName=Locomotion
    groupRole=AlwaysFollower
    method=SyncGroup
   }
   n29 = Blendspace Player 'AssaultRifle_Relaxed_CMF_AD'(X=@n2, Y=@n27) {
    blendSpace=&AssaultRifle_Relaxed_CMF_AD
    groupName=Locomotion
    groupRole=AlwaysFollower
    method=SyncGroup
   }
   n30 = "Blend Poses (EFortCustomGender)"(
    BlendPose_0=Blend Poses by bool(BlendPose_0=@n25, BlendPose_1=@n26, BlendTime_0=.4, BlendTime_1=0, bActiveValue=Get bIsRelaxedLevel1()) {blendType=Linear}
    BlendPose_1=Blend Poses by bool(BlendPose_0=@n29, BlendPose_1=@n28, BlendTime_0=.4, BlendTime_1=0, bActiveValue=Get bIsRelaxedLevel1()) {blendType=Linear}
    BlendTime_0=0.1
    BlendTime_1=0.1
    ActiveEnumValue=Get Gender() -> (Gender=Invalid)
   ) {blendType=Linear, blendPose [{linkId=-1, sourceLinkId=-1}, {linkId=-1, sourceLinkId=-1}], blendTime [0.1,0.1]}
   n31 = Get Speed()
   n32 = Blendspace Player 'M_Jog_ARDS_Additive'(X=@n2, Y=@n31) {
    blendSpace=/Game/Animation/Game/MainPlayer/Locomotion/Medium/Male/Jog/AssaultRifle_NonTargeted/M_Jog_ARDS_Additive
    groupName=Locomotion
    groupRole=AlwaysFollower
    method=SyncGroup
   }
   n33 = Blendspace Player 'M_Jog_ARRelaxed_Additive'(X=@n2, Y=@n31) {
    blendSpace=&M_Jog_ARRelaxed_Additive
    groupName=Locomotion
    groupRole=AlwaysFollower
    method=SyncGroup
   }
   n34 = Get Speed()
   n35 = Blendspace Player 'AssaultRifle_NonTargeted_CMF_Add'(X=@n2, Y=@n34) {
    blendSpace=/Game/Animation/Game/MainPlayer/Locomotion/Medium/Female/Jog/AssaultRifle_NonTargeted/AssaultRifle_NonTargeted_CMF_Add
    groupName=Locomotion
    groupRole=AlwaysFollower
    method=SyncGroup
   }
   n36 = Blendspace Player 'AssaultRifle_Relaxed_CMF_AD'(X=@n2, Y=@n34) {
    blendSpace=&AssaultRifle_Relaxed_CMF_AD
    groupName=Locomotion
    groupRole=AlwaysFollower
    method=SyncGroup
   }
   n37 = "Blend Poses (EFortCustomGender)"(
    BlendPose_0=Blend Poses by bool(BlendPose_0=@n33, BlendPose_1=@n32, BlendTime_0=.4, BlendTime_1=0, bActiveValue=Get bIsRelaxedLevel1()) {blendType=Linear}
    BlendPose_1=Blend Poses by bool(BlendPose_0=@n36, BlendPose_1=@n35, BlendTime_0=.4, BlendTime_1=0, bActiveValue=Get bIsRelaxedLevel1()) {blendType=Linear}
    BlendTime_0=0.1
    BlendTime_1=0.1
    ActiveEnumValue=Get Gender() -> (Gender=Invalid)
   ) {blendType=Linear, blendPose [{linkId=-1, sourceLinkId=-1}, {linkId=-1, sourceLinkId=-1}], blendTime [0.1,0.1]}
   n38 = Get Speed()
   n39 = Blendspace Player 'HeavyWeapon_AdditiveBS'(X=@n2, Y=@n38) {
    blendSpace=/Game/Animation/Game/MainPlayer/Locomotion/Medium/Male/Jog/HeavyWeapon/HeavyWeapon_AdditiveBS
    groupName=Locomotion
    groupRole=AlwaysFollower
    method=SyncGroup
   }
   n40 = Blendspace Player 'HeavyWeapon_AdditiveBS'(X=@n2, Y=@n38) {
    blendSpace=&HeavyWeapon_AdditiveBS
    groupName=Locomotion
    groupRole=AlwaysFollower
    method=SyncGroup
   }
   n41 = "Blend Poses (EFortCustomGender)"(BlendPose_0=@n39, BlendPose_1=@n40, BlendTime_0=0.1, BlendTime_1=0.1, ActiveEnumValue=Get Gender() -> (Gender=Invalid)) {blendType=Linear, blendPose [{linkId=-1, sourceLinkId=-1}, {linkId=-1, sourceLinkId=-1}], blendTime [0.1,0.1]}
   n42 = Get Speed()
   n43 = Blendspace Player 'M_Jog_ARDS_Additive'(X=@n2, Y=@n42) {
    blendSpace=&M_Jog_ARDS_Additive
    groupName=Locomotion
    groupRole=AlwaysFollower
    method=SyncGroup
   }
   n44 = Blendspace Player 'M_Jog_ARRelaxed_Additive'(X=@n2, Y=@n42) {
    blendSpace=&M_Jog_ARRelaxed_Additive
    groupName=Locomotion
    groupRole=AlwaysFollower
    method=SyncGroup
   }
   n45 = Get Speed()
   n46 = Blendspace Player 'AssaultRifle_NonTargeted_CMF_Add'(X=@n2, Y=@n45) {
    blendSpace=&AssaultRifle_NonTargeted_CMF_Add
    groupName=Locomotion
    groupRole=AlwaysFollower
    method=SyncGroup
   }
   n47 = Blendspace Player 'AssaultRifle_Relaxed_CMF_AD'(X=@n2, Y=@n45) {
    blendSpace=&AssaultRifle_Relaxed_CMF_AD
    groupName=Locomotion
    groupRole=AlwaysFollower
    method=SyncGroup
   }
   n48 = "Blend Poses (EFortCustomGender)"(
    BlendPose_0=Blend Poses by bool(BlendPose_0=@n44, BlendPose_1=@n43, BlendTime_0=.4, BlendTime_1=0, bActiveValue=Get bIsRelaxedLevel1()) {blendType=Linear}
    BlendPose_1=Blend Poses by bool(BlendPose_0=@n47, BlendPose_1=@n46, BlendTime_0=.4, BlendTime_1=0, bActiveValue=Get bIsRelaxedLevel1()) {blendType=Linear}
    BlendTime_0=0.1
    BlendTime_1=0.1
    ActiveEnumValue=Get Gender() -> (Gender=Invalid)
   ) {blendType=Linear, blendPose [{linkId=-1, sourceLinkId=-1}, {linkId=-1, sourceLinkId=-1}], blendTime [0.1,0.1]}
   n49 = Get Speed()
   n50 = Blendspace Player 'M_Jog_Melee_OH_Additive'(X=@n2, Y=@n49) {
    blendSpace=&M_Jog_Melee_OH_Additive
    groupName=Locomotion
    groupRole=AlwaysFollower
    method=SyncGroup
   }
   n51 = Blendspace Player 'Melee_OneHanded_CMF_ADD'(X=@n2, Y=@n49) {
    blendSpace=&Melee_OneHanded_CMF_ADD
    groupName=Locomotion
    groupRole=AlwaysFollower
    method=SyncGroup
   }
   n52 = "Blend Poses (EFortCustomGender)"(BlendPose_0=@n50, BlendPose_1=@n51, BlendTime_0=0.1, BlendTime_1=0.1, ActiveEnumValue=Get Gender() -> (Gender=Invalid)) {blendType=Linear, blendPose [{linkId=-1, sourceLinkId=-1}, {linkId=-1, sourceLinkId=-1}], blendTime [0.1,0.1]}
   n53 = Get Speed()
   n54 = Blendspace Player 'M_Jog_ARRelaxed_Additive'(X=@n2, Y=@n53) {
    blendSpace=&M_Jog_ARRelaxed_Additive
    groupName=Locomotion
    groupRole=AlwaysFollower
    method=SyncGroup
   }
   n55 = Blendspace Player 'M_Jog_AR_Additive'(X=@n2, Y=@n53) {
    blendSpace=&M_Jog_AR_Additive
    groupName=Locomotion
    groupRole=AlwaysFollower
    method=SyncGroup
   }
   n56 = Get Speed()
   n57 = Blendspace Player 'AssaultRifle_Relaxed_CMF_AD'(X=@n2, Y=@n56) {
    blendSpace=&AssaultRifle_Relaxed_CMF_AD
    groupName=Locomotion
    groupRole=AlwaysFollower
    method=SyncGroup
   }
   n58 = Blendspace Player 'AssaultRifle_HFNT_CMF_Add'(X=@n2, Y=@n56) {
    blendSpace=&AssaultRifle_HFNT_CMF_Add
    groupName=Locomotion
    groupRole=AlwaysFollower
    method=SyncGroup
   }
   n59 = "Blend Poses (EFortCustomGender)"(
    BlendPose_0=Blend Poses by bool(BlendPose_0=@n54, BlendPose_1=@n55, BlendTime_0=.4, BlendTime_1=0, bActiveValue=Get bIsRelaxedLevel1()) {blendType=Linear}
    BlendPose_1=Blend Poses by bool(BlendPose_0=@n57, BlendPose_1=@n58, BlendTime_0=.4, BlendTime_1=0, bActiveValue=Get bIsRelaxedLevel1()) {blendType=Linear}
    BlendTime_0=0.1
    BlendTime_1=0.1
    ActiveEnumValue=Get Gender() -> (Gender=Invalid)
   ) {blendType=Linear, blendPose [{linkId=-1, sourceLinkId=-1}, {linkId=-1, sourceLinkId=-1}], blendTime [0.1,0.1]}
   n60 = Get Speed()
   n61 = Blendspace Player 'M_Jog_ARDS_Additive'(X=@n2, Y=@n60) {
    blendSpace=&M_Jog_ARDS_Additive
    groupName=Locomotion
    groupRole=AlwaysFollower
    method=SyncGroup
   }
   n62 = Blendspace Player 'M_Jog_ARRelaxed_Additive'(X=@n2, Y=@n60) {
    blendSpace=&M_Jog_ARRelaxed_Additive
    groupName=Locomotion
    groupRole=AlwaysFollower
    method=SyncGroup
   }
   n63 = Get Speed()
   n64 = Blendspace Player 'AssaultRifle_NonTargeted_CMF_Add'(X=@n2, Y=@n63) {
    blendSpace=&AssaultRifle_NonTargeted_CMF_Add
    groupName=Locomotion
    groupRole=AlwaysFollower
    method=SyncGroup
   }
   n65 = Blendspace Player 'AssaultRifle_Relaxed_CMF_AD'(X=@n2, Y=@n63) {
    blendSpace=&AssaultRifle_Relaxed_CMF_AD
    groupName=Locomotion
    groupRole=AlwaysFollower
    method=SyncGroup
   }
   n66 = "Blend Poses (EFortCustomGender)"(
    BlendPose_0=Blend Poses by bool(BlendPose_0=@n62, BlendPose_1=@n61, BlendTime_0=.4, BlendTime_1=0, bActiveValue=Get bIsRelaxedLevel1()) {blendType=Linear}
    BlendPose_1=Blend Poses by bool(BlendPose_0=@n65, BlendPose_1=@n64, BlendTime_0=.4, BlendTime_1=0, bActiveValue=Get bIsRelaxedLevel1()) {blendType=Linear}
    BlendTime_0=0.1
    BlendTime_1=0.1
    ActiveEnumValue=Get Gender() -> (Gender=Invalid)
   ) {blendType=Linear, blendPose [{linkId=-1, sourceLinkId=-1}, {linkId=-1, sourceLinkId=-1}], blendTime [0.1,0.1]}
   n67 = Get Speed()
   n68 = Blendspace Player 'NoWep_CMF_ADD'(X=@n2, Y=@n67) {
    blendSpace=&NoWep_CMF_ADD
    groupName=Locomotion
    groupRole=AlwaysFollower
    method=SyncGroup
   }
   n69 = Blendspace Player 'M_Jog_NoWep_Additive'(X=@n2, Y=@n67) {
    blendSpace=&M_Jog_NoWep_Additive
    groupName=Locomotion
    groupRole=AlwaysFollower
    method=SyncGroup
   }
   n70 = "Blend Poses (EFortCustomGender)"(BlendPose_0=@n69, BlendPose_1=@n68, BlendTime_0=0.1, BlendTime_1=0.1, ActiveEnumValue=Get Gender() -> (Gender=Invalid)) {blendType=Linear, blendPose [{linkId=-1, sourceLinkId=-1}, {linkId=-1, sourceLinkId=-1}], blendTime [0.1,0.1]}
   n71 = Get Speed()
   n72 = Blendspace Player 'M_Jog_Melee_TH_Additive'(X=@n2, Y=@n71) {
    blendSpace=/Game/Animation/Game/MainPlayer/Locomotion/Medium/Male/Jog/Melee_TH/M_Jog_Melee_TH_Additive
    groupName=Locomotion
    groupRole=AlwaysFollower
    method=SyncGroup
   }
   n73 = Blendspace Player 'Melee_TwoHanded_CMF_ADD'(X=@n2, Y=@n71) {
    blendSpace=/Game/Animation/Game/MainPlayer/Locomotion/Medium/Female/Jog/Melee_TwoHanded/Melee_TwoHanded_CMF_ADD
    groupName=Locomotion
    groupRole=AlwaysFollower
    method=SyncGroup
   }
   n74 = "Blend Poses (EFortCustomGender)"(BlendPose_0=@n72, BlendPose_1=@n73, BlendTime_0=0, BlendTime_1=0, ActiveEnumValue=Get Gender() -> (Gender=Invalid)) {blendType=Linear, blendPose [{linkId=-1, sourceLinkId=-1}, {linkId=-1, sourceLinkId=-1}], blendTime [0.1,0.1]}
   n75 = Get Speed()
   n76 = Blendspace Player 'NoWep_CMF_ADD'(X=@n2, Y=@n75) {
    blendSpace=&NoWep_CMF_ADD
    groupName=Locomotion
    groupRole=AlwaysFollower
    method=SyncGroup
   }
   n77 = Blendspace Player 'M_Jog_NoWep_Additive'(X=@n2, Y=@n75) {
    blendSpace=&M_Jog_NoWep_Additive
    groupName=Locomotion
    groupRole=AlwaysFollower
    method=SyncGroup
   }
   n78 = "Blend Poses (EFortCustomGender)"(BlendPose_0=@n77, BlendPose_1=@n76, BlendTime_0=0, BlendTime_1=0, ActiveEnumValue=Get Gender() -> (Gender=Invalid)) {blendType=Linear, blendPose [{linkId=-1, sourceLinkId=-1}, {linkId=-1, sourceLinkId=-1}], blendTime [0.1,0.1]}
   n79 = Blendspace Player 'M_Jog_ARDS_Additive'(X=@n2, Y=@n1) {
    blendSpace=&M_Jog_ARDS_Additive
    groupName=Locomotion
    groupRole=AlwaysFollower
    method=SyncGroup
   }
   n80 = Blendspace Player 'AssaultRifle_NonTargeted_CMF_Add'(X=@n2, Y=@n4) {
    blendSpace=&AssaultRifle_NonTargeted_CMF_Add
    groupName=Locomotion
    groupRole=AlwaysFollower
    method=SyncGroup
   }
   n81 = "Blend Poses (EFortCustomGender)"(BlendPose_0=@n79, BlendPose_1=@n80, BlendTime_0=0.1, BlendTime_1=0.1, ActiveEnumValue=Get Gender() -> (Gender=Invalid)) {blendType=Linear, blendPose [{linkId=-1, sourceLinkId=-1}, {linkId=-1, sourceLinkId=-1}], blendTime [0.1,0.1]}
   n82 = Get Speed()
   n83 = Blendspace Player 'M_Jog_ARRelaxed_Additive'(X=@n2, Y=@n82) {
    blendSpace=&M_Jog_ARRelaxed_Additive
    groupName=Locomotion
    groupRole=AlwaysFollower
    method=SyncGroup
   }
   n84 = Blendspace Player 'M_Jog_ARDS_Additive'(X=@n2, Y=@n82) {
    blendSpace=&M_Jog_ARDS_Additive
    groupName=Locomotion
    groupRole=AlwaysFollower
    method=SyncGroup
   }
   n85 = Get Speed()
   n86 = Blendspace Player 'AssaultRifle_Relaxed_CMF_AD'(X=@n2, Y=@n85) {
    blendSpace=&AssaultRifle_Relaxed_CMF_AD
    groupName=Locomotion
    groupRole=AlwaysFollower
    method=SyncGroup
   }
   n87 = Blendspace Player 'AssaultRifle_NonTargeted_CMF_Add'(X=@n2, Y=@n85) {
    blendSpace=&AssaultRifle_NonTargeted_CMF_Add
    groupName=Locomotion
    groupRole=AlwaysFollower
    method=SyncGroup
   }
   n88 = "Blend Poses (EFortCustomGender)"(
    BlendPose_0=Blend Poses by bool(BlendPose_0=@n83, BlendPose_1=@n84, BlendTime_0=.4, BlendTime_1=0, bActiveValue=Get bIsRelaxedLevel1()) {blendType=Linear}
    BlendPose_1=Blend Poses by bool(BlendPose_0=@n86, BlendPose_1=@n87, BlendTime_0=.4, BlendTime_1=0, bActiveValue=Get bIsRelaxedLevel1()) {blendType=Linear}
    BlendTime_0=0.1
    BlendTime_1=0.1
    ActiveEnumValue=Get Gender() -> (Gender=Invalid)
   ) {blendType=Linear, blendPose [{linkId=-1, sourceLinkId=-1}, {linkId=-1, sourceLinkId=-1}], blendTime [0.1,0.1]}
   n89 = Get Speed()
   n90 = Blendspace Player 'Launcher_Jog_AdditiveBS'(X=@n2, Y=@n89) {
    blendSpace=/Game/Animation/Game/MainPlayer/Locomotion/Medium/Male/Jog/Launcher/Launcher_Jog_AdditiveBS
    groupName=Locomotion
    groupRole=AlwaysFollower
    method=SyncGroup
   }
   n91 = Blendspace Player 'Launcher_Jog_AdditiveBS'(X=@n2, Y=@n89) {
    blendSpace=&Launcher_Jog_AdditiveBS
    groupName=Locomotion
    groupRole=AlwaysFollower
    method=SyncGroup
   }
   n92 = "Blend Poses (EFortCustomGender)"(BlendPose_0=@n90, BlendPose_1=@n91, BlendTime_0=0.1, BlendTime_1=0.1, ActiveEnumValue=Get Gender() -> (Gender=Invalid)) {blendType=Linear, blendPose [{linkId=-1, sourceLinkId=-1}, {linkId=-1, sourceLinkId=-1}], blendTime [0.1,0.1]}
   n93 = Get Speed()
   n94 = Blendspace Player 'M_Jog_ARRelaxed_Additive'(X=@n2, Y=@n93) {
    blendSpace=&M_Jog_ARRelaxed_Additive
    groupName=Locomotion
    groupRole=AlwaysFollower
    method=SyncGroup
   }
   n95 = Blendspace Player 'M_Jog_ARDS_Additive'(X=@n2, Y=@n93) {
    blendSpace=&M_Jog_ARDS_Additive
    groupName=Locomotion
    groupRole=AlwaysFollower
    method=SyncGroup
   }
   n96 = Get Speed()
   n97 = Blendspace Player 'M_Jog_ARRelaxed_Additive'(X=@n2, Y=@n96) {
    blendSpace=&M_Jog_ARRelaxed_Additive
    groupName=Locomotion
    groupRole=AlwaysFollower
    method=SyncGroup
   }
   n98 = Blendspace Player 'M_Jog_ARDS_Additive'(X=@n2, Y=@n96) {
    blendSpace=&M_Jog_ARDS_Additive
    groupName=Locomotion
    groupRole=AlwaysFollower
    method=SyncGroup
   }
   n99 = "Blend Poses (EFortCustomGender)"(
    BlendPose_0=Blend Poses by bool(BlendPose_0=@n94, BlendPose_1=@n95, BlendTime_0=.4, BlendTime_1=0, bActiveValue=Get bIsRelaxedLevel1()) {blendType=Linear}
    BlendPose_1=Blend Poses by bool(BlendPose_0=@n97, BlendPose_1=@n98, BlendTime_0=.4, BlendTime_1=0, bActiveValue=Get bIsRelaxedLevel1()) {blendType=Linear}
    BlendTime_0=0.1
    BlendTime_1=0.1
    ActiveEnumValue=Get Gender() -> (Gender=Invalid)
   ) {blendType=Linear, blendPose [{linkId=-1, sourceLinkId=-1}, {linkId=-1, sourceLinkId=-1}], blendTime [0.1,0.1]}
   n100 = Get Speed()
   n101 = Blendspace Player 'M_Jog_ARRelaxed_Additive'(X=@n2, Y=@n100) {
    blendSpace=&M_Jog_ARRelaxed_Additive
    groupName=Locomotion
    groupRole=AlwaysFollower
    method=SyncGroup
   }
   n102 = Blendspace Player 'M_Jog_ARDS_Additive'(X=@n2, Y=@n100) {
    blendSpace=&M_Jog_ARDS_Additive
    groupName=Locomotion
    groupRole=AlwaysFollower
    method=SyncGroup
   }
   n103 = Get Speed()
   n104 = Blendspace Player 'AssaultRifle_Relaxed_CMF_AD'(X=@n2, Y=@n103) {
    blendSpace=&AssaultRifle_Relaxed_CMF_AD
    groupName=Locomotion
    groupRole=AlwaysFollower
    method=SyncGroup
   }
   n105 = Blendspace Player 'AssaultRifle_NonTargeted_CMF_Add'(X=@n2, Y=@n103) {
    blendSpace=&AssaultRifle_NonTargeted_CMF_Add
    groupName=Locomotion
    groupRole=AlwaysFollower
    method=SyncGroup
   }
   n106 = "Blend Poses (EFortCustomGender)"(
    BlendPose_0=Blend Poses by bool(BlendPose_0=@n101, BlendPose_1=@n102, BlendTime_0=.4, BlendTime_1=0, bActiveValue=Get bIsRelaxedLevel1()) {blendType=Linear}
    BlendPose_1=Blend Poses by bool(BlendPose_0=@n104, BlendPose_1=@n105, BlendTime_0=.4, BlendTime_1=0, bActiveValue=Get bIsRelaxedLevel1()) {blendType=Linear}
    BlendTime_0=0.1
    BlendTime_1=0.1
    ActiveEnumValue=Get Gender() -> (Gender=Invalid)
   ) {blendType=Linear, blendPose [{linkId=-1, sourceLinkId=-1}, {linkId=-1, sourceLinkId=-1}], blendTime [0.1,0.1]}
   n107 = Get Speed()
   n108 = Blendspace Player 'M_Jog_Melee_TH_Additive'(X=@n2, Y=@n107) {
    blendSpace=&M_Jog_Melee_TH_Additive
    groupName=Locomotion
    groupRole=AlwaysFollower
    method=SyncGroup
   }
   n109 = Blendspace Player 'Melee_TwoHanded_CMF_ADD'(X=@n2, Y=@n107) {
    blendSpace=&Melee_TwoHanded_CMF_ADD
    groupName=Locomotion
    groupRole=AlwaysFollower
    method=SyncGroup
   }
   n110 = "Blend Poses (EFortCustomGender)"(BlendPose_0=@n108, BlendPose_1=@n109, BlendTime_0=0.1, BlendTime_1=0.1, ActiveEnumValue=Get Gender() -> (Gender=Invalid)) {blendType=Linear, blendPose [{linkId=-1, sourceLinkId=-1}, {linkId=-1, sourceLinkId=-1}], blendTime [0.1,0.1]}
   n111 = Get Speed()
   n112 = Blendspace Player 'M_Jog_ARRelaxed_Additive'(X=@n2, Y=@n111) {
    blendSpace=&M_Jog_ARRelaxed_Additive
    groupName=Locomotion
    groupRole=AlwaysFollower
    method=SyncGroup
   }
   n113 = Blendspace Player 'M_Jog_AR_Additive'(X=@n2, Y=@n111) {
    blendSpace=&M_Jog_AR_Additive
    groupName=Locomotion
    groupRole=AlwaysFollower
    method=SyncGroup
   }
   n114 = Get Speed()
   n115 = Blendspace Player 'M_Jog_ARRelaxed_Additive'(X=@n2, Y=@n114) {
    blendSpace=&M_Jog_ARRelaxed_Additive
    groupName=Locomotion
    groupRole=AlwaysFollower
    method=SyncGroup
   }
   n116 = Blendspace Player 'AssaultRifle_HFNT_CMF_Add'(X=@n2, Y=@n114) {
    blendSpace=&AssaultRifle_HFNT_CMF_Add
    groupName=Locomotion
    groupRole=AlwaysFollower
    method=SyncGroup
   }
   n117 = "Blend Poses (EFortCustomGender)"(
    BlendPose_0=Blend Poses by bool(BlendPose_0=@n112, BlendPose_1=@n113, BlendTime_0=.4, BlendTime_1=0, bActiveValue=Get bIsRelaxedLevel1()) {blendType=Linear}
    BlendPose_1=Blend Poses by bool(BlendPose_0=@n115, BlendPose_1=@n116, BlendTime_0=.4, BlendTime_1=0, bActiveValue=Get bIsRelaxedLevel1()) {blendType=Linear}
    BlendTime_0=0.1
    BlendTime_1=0.1
    ActiveEnumValue=Get Gender() -> (Gender=Invalid)
   ) {blendType=Linear, blendPose [{linkId=-1, sourceLinkId=-1}, {linkId=-1, sourceLinkId=-1}], blendTime [0.1,0.1]}
   n118 = Get Speed()
   n119 = Blendspace Player 'M_Jog_ARRelaxed_Additive'(X=@n2, Y=@n118) {
    blendSpace=&M_Jog_ARRelaxed_Additive
    groupName=Locomotion
    groupRole=AlwaysFollower
    method=SyncGroup
   }
   n120 = Blendspace Player 'M_Jog_ARDS_Additive'(X=@n2, Y=@n118) {
    blendSpace=&M_Jog_ARDS_Additive
    groupName=Locomotion
    groupRole=AlwaysFollower
    method=SyncGroup
   }
   n121 = Get Speed()
   n122 = Blendspace Player 'AssaultRifle_Relaxed_CMF_AD'(X=@n2, Y=@n121) {
    blendSpace=&AssaultRifle_Relaxed_CMF_AD
    groupName=Locomotion
    groupRole=AlwaysFollower
    method=SyncGroup
   }
   n123 = Blendspace Player 'AssaultRifle_NonTargeted_CMF_Add'(X=@n2, Y=@n121) {
    blendSpace=&AssaultRifle_NonTargeted_CMF_Add
    groupName=Locomotion
    groupRole=AlwaysFollower
    method=SyncGroup
   }
   n124 = "Blend Poses (EFortCustomGender)"(
    BlendPose_0=Blend Poses by bool(BlendPose_0=@n119, BlendPose_1=@n120, BlendTime_0=.4, BlendTime_1=0, bActiveValue=Get bIsRelaxedLevel1()) {blendType=Linear}
    BlendPose_1=Blend Poses by bool(BlendPose_0=@n122, BlendPose_1=@n123, BlendTime_0=.4, BlendTime_1=0, bActiveValue=Get bIsRelaxedLevel1()) {blendType=Linear}
    BlendTime_0=0.1
    BlendTime_1=0.1
    ActiveEnumValue=Get Gender() -> (Gender=Invalid)
   ) {blendType=Linear, blendPose [{linkId=-1, sourceLinkId=-1}, {linkId=-1, sourceLinkId=-1}], blendTime [0.1,0.1]}
   n125 = Get Speed()
   n126 = Blendspace Player 'ExplosiveBow_Relaxed_Jog_CMM_Additive'(X=@n2, Y=@n125) {
    blendSpace=/Game/Animation/Game/MainPlayer/Locomotion/Medium/Male/Jog/ExplosiveBow/ExplosiveBow_Relaxed_Jog_CMM_Additive
    groupName=Locomotion
    groupRole=AlwaysFollower
    method=SyncGroup
   }
   n127 = Blendspace Player 'ExplosiveBow_Jog_CMM_Additive'(X=@n2, Y=@n125) {
    blendSpace=/Game/Animation/Game/MainPlayer/Locomotion/Medium/Male/Jog/ExplosiveBow/ExplosiveBow_Jog_CMM_Additive
    groupName=Locomotion
    groupRole=AlwaysFollower
    method=SyncGroup
   }
   n128 = Get Speed()
   n129 = Blendspace Player 'ExplosiveBow_Relaxed_Jog_CMF_Additive'(X=@n2, Y=@n128) {
    blendSpace=/Game/Animation/Game/MainPlayer/Locomotion/Medium/Female/Jog/ExplosiveBow/ExplosiveBow_Relaxed_Jog_CMF_Additive
    groupName=Locomotion
    groupRole=AlwaysFollower
    method=SyncGroup
   }
   n130 = Blendspace Player 'ExplosiveBow_Jog_CMF_Additive'(X=@n2, Y=@n128) {
    blendSpace=/Game/Animation/Game/MainPlayer/Locomotion/Medium/Female/Jog/ExplosiveBow/ExplosiveBow_Jog_CMF_Additive
    groupName=Locomotion
    groupRole=AlwaysFollower
    method=SyncGroup
   }
   n131 = "Blend Poses (EFortCustomGender)"(
    BlendPose_0=Blend Poses by bool(BlendPose_0=@n126, BlendPose_1=@n127, BlendTime_0=.4, BlendTime_1=0, bActiveValue=Get bIsRelaxedLevel1()) {blendType=Linear}
    BlendPose_1=Blend Poses by bool(BlendPose_0=@n129, BlendPose_1=@n130, BlendTime_0=.4, BlendTime_1=0, bActiveValue=Get bIsRelaxedLevel1()) {blendType=Linear}
    BlendTime_0=0.1
    BlendTime_1=0.1
    ActiveEnumValue=Get Gender() -> (Gender=Invalid)
   ) {blendType=Linear, blendPose [{linkId=-1, sourceLinkId=-1}, {linkId=-1, sourceLinkId=-1}], blendTime [0.1,0.1]}
   n132 = Local To Component(LocalPose=@n131)
   n133 = "Blend Poses (EFortWeaponCoreAnimation)"!(
    BlendPose_0=@n70
    BlendPose_1=Blend Poses by bool(
     BlendPose_0=Blend Poses by bool(BlendPose_0=@n15, BlendPose_1=@n18, BlendTime_0=0.4, BlendTime_1=0, bActiveValue=@n19) {blendType=Linear}
     BlendPose_1=@n23
     BlendTime_0=.4
     BlendTime_1=0
     bActiveValue=Get bIsRelaxedLevel1()
    ) {blendType=Linear}
    BlendPose_2=@n110
    BlendPose_3=@n52
    BlendPose_4=@n41
    BlendPose_5=@n48
    BlendPose_6=@n59
    BlendPose_7=@n37
    BlendPose_8=@n66
    BlendPose_9=@n74
    BlendPose_10=@n78
    BlendPose_11=@n81
    BlendPose_12=@n30
    BlendPose_13=@n92
    BlendPose_14=@n106
    BlendPose_15=@n99
    BlendPose_16=@n117
    BlendPose_17:Pose Link Structure=@n88
    BlendPose_18:Pose Link Structure=@n124
    BlendPose_19:Pose Link Structure=Component To Local(ComponentPose=Copy Bone - Source Bone: weapon_l - Target Bone: weapon_r!(ComponentPose=@n132, Alpha:float=1, bCopyTranslation=true, bCopyRotation=true, bCopyScale=true) -> (Pose) {alphaBoolBlend {blendOption=Linear}, sourceBone {boneName=weapon_l}, targetBone {boneName=weapon_r}})
    BlendTime_0=0
    BlendTime_1=0
    BlendTime_2=0.1
    BlendTime_3=0.1
    BlendTime_4=0.1
    BlendTime_5=0.1
    BlendTime_6=0.1
    BlendTime_7=0.1
    BlendTime_8=0.1
    BlendTime_9=0.1
    BlendTime_10=0.1
    BlendTime_11=0.1
    BlendTime_12=0.1
    BlendTime_13=0.1
    BlendTime_14=0.1
    BlendTime_15=0.1
    BlendTime_16=0.1
    BlendTime_17:float=0.1
    BlendTime_18:float=0.1
    BlendTime_19:float=0.1
    ActiveEnumValue=Get WeaponCoreAnimation()
   ) -> (Pose) {
    blendPose [
     {linkId=-1, sourceLinkId=-1}
     {linkId=-1, sourceLinkId=-1}
     {linkId=-1, sourceLinkId=-1}
     {linkId=-1, sourceLinkId=-1}
     {linkId=-1, sourceLinkId=-1}
     {linkId=-1, sourceLinkId=-1}
     {linkId=-1, sourceLinkId=-1}
     {linkId=-1, sourceLinkId=-1}
     {linkId=-1, sourceLinkId=-1}
     {linkId=-1, sourceLinkId=-1}
     {linkId=-1, sourceLinkId=-1}
     {linkId=-1, sourceLinkId=-1}
     {linkId=-1, sourceLinkId=-1}
     {linkId=-1, sourceLinkId=-1}
     {linkId=-1, sourceLinkId=-1}
     {linkId=-1, sourceLinkId=-1}
     {linkId=-1, sourceLinkId=-1}
     {linkId=-1, sourceLinkId=-1}
     {linkId=-1, sourceLinkId=-1}
     {linkId=-1, sourceLinkId=-1}
    ]
    blendTime [0.1,0.1,0.1,0.1,0.1,0.1,0.1,0.1,0.1,0.1,0.1,0.1,0.1,0.1,0.1,0.1,0.1,0.1,0.1,0.1]
   }
   n134 = Get Speed()
   n135 = Get MovementDirection()
   n136 = Blendspace Player 'M_Locomotion2D_Walk_AR_Additive'(X=@n135, Y=@n134) {
    blendSpace=/Game/Animation/Game/MainPlayer/Locomotion/Medium/Male/Walk/AssaultRifle_Targeted/M_Locomotion2D_Walk_AR_Additive
    groupName=Locomotion
    groupRole=AlwaysFollower
    method=SyncGroup
   }
   n137 = Blendspace Player 'M_Locomotion2D_Walk_Pistol_Additive'(X=@n135, Y=@n134) {
    blendSpace=/Game/Animation/Game/MainPlayer/Locomotion/Medium/Male/Walk/Pistol/M_Locomotion2D_Walk_Pistol_Additive
    groupName=Locomotion
    groupRole=AlwaysFollower
    method=SyncGroup
   }
   n138 = "Blend Poses (EFortWeaponCoreAnimation)"!(BlendPose_0=@n136, BlendPose_1=@n137, BlendTime_0=0, BlendTime_1=0, ActiveEnumValue=Get WeaponCoreAnimation()) -> (Pose) {blendType=Linear, blendPose [{linkId=-1, sourceLinkId=-1}, {linkId=-1, sourceLinkId=-1}], blendTime [0.1,0.1]}
   n139 = "Blend Poses (EFortMovementStyle)"(BlendPose_0=@n133, BlendPose_1=@n138, BlendTime_0=0.1, BlendTime_1=0.1, ActiveEnumValue=Get MovementStyle() -> (MovementStyle=Running)) {blendType=Linear, blendPose [{linkId=-1, sourceLinkId=-1}, {linkId=-1, sourceLinkId=-1}], blendTime [0.1,0.1]}
   Output Animation Pose(Result=Blend Poses by bool(BlendPose_0=@n139, BlendPose_1=@n11, bActiveValue=Get bHasWeapon()) {blendType=Linear})
   Get Gender() -> (Gender=Invalid)
  ]
 }
 Transition {
  path=AnimGraph.AnimGraphNode_StateMachine_55.SurvivorFullBody.AnimStateNode_2.SplitBody.AnimGraphNode_StateMachine_87.LocomotionAdditive.AnimStateTransitionNode_2.Transition
  nodes [Result(bCanEnterTransition=Get bShouldUseMovementLocomotion() -> (bShouldUseMovementLocomotion:bool))]
 }
 Transition {
  path=AnimGraph.AnimGraphNode_StateMachine_55.SurvivorFullBody.AnimStateNode_2.SplitBody.AnimGraphNode_StateMachine_87.LocomotionAdditive.AnimStateTransitionNode_3.Transition
  nodes [Result(bCanEnterTransition=Get bTransitionToIdle() -> (bTransitionToIdle:bool))]
 }
 DBNO {
  path=AnimGraph.AnimGraphNode_StateMachine_55.SurvivorFullBody.AnimStateNode_3.DBNO
  nodes [Output Animation Pose(Result=Sequence Player 'DBNO_Loop_CMM'() {sequence=/Game/Animation/Game/MainPlayer/Reacts/HitReacts/DBNO_Loop_CMM})]
 }
 Cower {
  path=AnimGraph.AnimGraphNode_StateMachine_55.SurvivorFullBody.AnimStateNode_9.Cower
  nodes [Output Animation Pose(Result=Sequence Player 'M_JumpApex_ARDS_Loop_Copy'() {sequence=/Game/Animation/Game/MainPlayer/Locomotion/Medium/Male/Jump/AssaultRifle/M_JumpApex_ARDS_Loop_Copy})]
 }
 Transition {
  path=AnimGraph.AnimGraphNode_StateMachine_55.SurvivorFullBody.AnimStateTransitionNode_11.Transition
  nodes [Result(bCanEnterTransition=Get bIsCowering() -> (bIsCowering:bool))]
 }
 Transition {
  path=AnimGraph.AnimGraphNode_StateMachine_55.SurvivorFullBody.AnimStateTransitionNode_12.Transition
  nodes [Result(bCanEnterTransition=Get bIsStunned() -> (bIsStunned:bool))]
 }
 Transition {
  path=AnimGraph.AnimGraphNode_StateMachine_55.SurvivorFullBody.AnimStateTransitionNode_13.Transition
  nodes [Result(bCanEnterTransition=Get bIsDBNO())]
 }
 Transition {
  path=AnimGraph.AnimGraphNode_StateMachine_55.SurvivorFullBody.AnimStateTransitionNode_14.Transition
  nodes [Result(bCanEnterTransition=NOT Boolean(A=Get bIsStunned()))]
 }
 Transition {
  path=AnimGraph.AnimGraphNode_StateMachine_55.SurvivorFullBody.AnimStateTransitionNode_7.Transition
  nodes [Result(bCanEnterTransition=NOT Boolean(A=Get bIsCowering()))]
 }
 LocomotionLowerBody {
  entry=Idle
  path=AnimGraph.AnimGraphNode_StateMachine_56.LocomotionLowerBody
  ids=id
  nodes [
   AnimStateNode_0 = Idle()
   AnimStateNode_1 = Movement() {is=title AnimStateNode}
   AnimStateTransitionNode_0 = AnimStateTransitionNode() {crossfadeDuration=0.3}
   AnimStateTransitionNode_1 = AnimStateTransitionNode() {crossfadeDuration=0.3}
  ]
  links [
   {
    dst=Movement
    rule=AnimGraph.AnimGraphNode_StateMachine_56.LocomotionLowerBody.AnimStateTransitionNode_0.Transition
    src=Idle
    via=AnimStateTransitionNode_0
   }
   {
    dst=Idle
    rule=AnimGraph.AnimGraphNode_StateMachine_56.LocomotionLowerBody.AnimStateTransitionNode_1.Transition
    src=Movement
    via=AnimStateTransitionNode_1
   }
  ]
 }
 Idle {
  path=AnimGraph.AnimGraphNode_StateMachine_56.LocomotionLowerBody.AnimStateNode_0.Idle
  nodes [Output Animation Pose(Result=Use cached pose 'UpperBody'())]
 }
 Movement {
  path=AnimGraph.AnimGraphNode_StateMachine_56.LocomotionLowerBody.AnimStateNode_1.Movement
  nodes [
   n1 = Get Speed()
   n2 = Get MovementDirection()
   n3 = Get SpeedAdjustedPlayRate() -> (SpeedAdjustedPlayRate:float)
   n4 = Blendspace Player 'M_Locomotion2D_Walk_NoWeapon'!(X=@n2, Y=@n1, PlayRate:float=@n3) -> (Pose) {
    blendSpace=/Game/Animation/Game/MainPlayer/Locomotion/Medium/Male/Walk/No_Weapon/M_Locomotion2D_Walk_NoWeapon
    groupName=Locomotion
    method=SyncGroup
    startPosition=0.7
   }
   n5 = Blendspace Player 'Survivor_Jog_Default'!(X=@n2, Y=@n1, PlayRate:float=@n3) -> (Pose) {
    blendSpace=/Game/Animation/Game/Survivor/Locomotion/Jog/Default/Survivor_Jog_Default
    groupName=Locomotion
    method=SyncGroup
    startPosition=0.72
   }
   n6 = "Blend Poses (EFortMovementStyle)"(BlendPose_0=@n5, BlendPose_1=@n4, BlendTime_0=0.1, BlendTime_1=0.1, ActiveEnumValue=Get MovementStyle() -> (MovementStyle=Running)) {blendType=Linear, blendPose [{linkId=-1, sourceLinkId=-1}, {linkId=-1, sourceLinkId=-1}], blendTime [0.1,0.1]}
   n7 = Blendspace Player 'M_Locomotion2D_Walk_AR_Core'!(X=@n2, Y=@n1, PlayRate:float=@n3) -> (Pose) {
    blendSpace=/Game/Animation/Game/MainPlayer/Locomotion/Medium/Male/Walk/AssaultRifle_Targeted/M_Locomotion2D_Walk_AR_Core
    groupName=Locomotion
    method=SyncGroup
    startPosition=0.7
   }
   n8 = Blendspace Player 'M_Locomotion2D_Walk_AR_Core_RtFootFwd'!(X=@n2, Y=@n1, PlayRate:float=@n3) -> (Pose) {
    blendSpace=/Game/Animation/Game/MainPlayer/Locomotion/Medium/Male/Walk/AssaultRifle_Targeted/M_Locomotion2D_Walk_AR_Core_RtFootFwd
    groupName=Locomotion
    method=SyncGroup
    startPosition=0.7
   }
   n9 = Blendspace Player 'M_Jog_AR_Core'!(X=@n2, Y=@n1, PlayRate:float=@n3) -> (Pose) {
    blendSpace=/Game/Animation/Game/MainPlayer/Locomotion/Medium/Male/Jog/AssaultRifle_HipNonTargeted/M_Jog_AR_Core
    groupName=Locomotion
    method=SyncGroup
    startPosition=0.72
   }
   n10 = "Blend Poses (EFortMovementStyle)"(
    BlendPose_0=Blend Poses by bool(BlendPose_0=@n8, BlendPose_1=@n7, bActiveValue=Get bShouldWalkRightFootForward() -> (bShouldWalkRightFootForward:bool)) {blendType=Linear}
    BlendPose_1=@n9
    BlendTime_0=0.1
    BlendTime_1=0.1
    ActiveEnumValue=Get MovementStyle() -> (MovementStyle=Running)
   ) {blendType=Linear, blendPose [{linkId=-1, sourceLinkId=-1}, {linkId=-1, sourceLinkId=-1}], blendTime [0.1,0.1]}
   Output Animation Pose(Result=Blend Poses by bool(BlendPose_0=@n10, BlendPose_1=@n6, bActiveValue=Get bHasWeapon()) {blendType=Linear})
  ]
 }
 Transition {
  path=AnimGraph.AnimGraphNode_StateMachine_56.LocomotionLowerBody.AnimStateTransitionNode_0.Transition
  nodes [Result(bCanEnterTransition=Get bShouldUseMovementLocomotion())]
 }
 Transition {
  path=AnimGraph.AnimGraphNode_StateMachine_56.LocomotionLowerBody.AnimStateTransitionNode_1.Transition
  nodes [Result(bCanEnterTransition=Get bTransitionToIdle())]
 }
 EventGraph {nodes [
   n1 = ParseWeaponType!() -> (OutputDelegate, then, EquippedWeapon:Fort Weapon Object Reference) {is=title K2Node_CustomEvent}
   > Set UsingBoomBow(UsingBoomBow:bool="Equal (Object)"(A=GetWeaponPawnAnimSet(self=@n1.EquippedWeapon), B=/Game/Animation/Game/MainPlayer/AnimAssets/ExplosiveBow_AnimAsset)) -> (Output_Get:bool)
   > Set UsingSpearWithThrust(UsingSpearWithThrust:bool="Equal (Object)"(A=GetWeaponPawnAnimSet(self=@n1.EquippedWeapon), B=&Spear_AnimAsset)) -> (Output_Get:bool)
   > Set Is Using Two Handed Sword(Is Using Two Handed Sword:bool="Equal (Object)"(A=GetWeaponPawnAnimSet(self=@n1.EquippedWeapon), B=/Game/Animation/Game/MainPlayer/AnimAssets/HeavySword_AnimAsset)) -> (Output_Get:bool)
  ]}
]
variables [
 Is Using Two Handed Sword {type=boolean}
 UsingBoomBow {type=boolean}
 UsingSpearWithThrust {type=boolean}
]
